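# Physics asset for the belt mesh of the male medium CrownOrder player cosmetic, covering a short hanging belt chain.
# Capsule bodies form a chain from base_beltChain through two dyn_beltChain bones, joined by two constraints with twist locked and uneven swing limits that loosen toward the tip.
# A larger, unconstrained right-thigh capsule is also included, apparently as a collider for the chain.
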PhysicsAsset /BRCosmetics/Characters/Player/Male/Medium/Bodies/M_MED_CrownOrder/Meshes/M_MED_CrownOrder_Belt_Physics
bodies {
 base_beltChain [{shapeType=Capsule, length=0.7863, radius=0.7101, shapeName=base_beltChain_capsule}]
 dyn_beltChain_01 [{shapeType=Capsule, length=6.2131, radius=0.5, shapeName=dyn_beltChain_01_capsule, center=<1.2 0 -0.6>}]
 dyn_beltChain_02 [{shapeType=Capsule, length=2.8798, radius=0.5, shapeName=dyn_beltChain_02_capsule, center=<0.6 0 -1>}]
 thigh_r [{
   shapeType=Capsule
   length=12.6955
   radius=8.84
   shapeName=thigh_r_capsule
   center=<5.582 0.3801 -1.6682>
   rotation=<rot 70.216 90.2664 92.1218>
  }]
}
constraints [
 {bone1Name=dyn_beltChain_01, bone2Name=base_beltChain, swing1LimitDegrees=30, swing2LimitDegrees=5, twistLimitDegrees=45, twistMotion=Locked}
 {
  bone1Name=dyn_beltChain_02
  bone2Name=dyn_beltChain_01
  swing1LimitDegrees=45
  swing2LimitDegrees=15
  twistLimitDegrees=45
  twistMotion=Locked
 }
]
