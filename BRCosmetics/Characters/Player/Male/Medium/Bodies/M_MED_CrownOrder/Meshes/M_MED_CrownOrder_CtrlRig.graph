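# Control Rig for the M_MED_CrownOrder body that adds corrective motion to its shoulder chains, belt chain and collar.
# It runs a shared deform rig, then ShoulderChains, HipChain and FluffAdjust, plus a front-end-only shoulder tweak.
# Weighted parent constraints place base bones, and offsets are blended by swing/twist angle remaps, spherical pose readers and a deform-rig curve.
# Interpolated movement-state bools, such as crouching and hang gliding, blend further offsets.
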ControlRigBlueprint /BRCosmetics/Characters/Player/Male/Medium/Bodies/M_MED_CrownOrder/Meshes/M_MED_CrownOrder_CtrlRig
graphs [
 CollapseNode_ContainedGraph {nodes [
   n1 = RigVMFunctionEntryNode()
   n2 = &RigVMFunction_AlphaInterp(
    InterpSpeedDecreasing=10
    InterpSpeedIncreasing=10
    Scale=1
    bClampResult=true
    Value=&RigVMFunction_MathBoolToFloat(Value=RigVMVariableNode() {variable=Crouching}.Value).Result
   )
   n3 = &RigVMFunction_MathDoubleMul(B=@n2.Result, A=@n1.ADS Up Pistol)
   &RigUnit_ModifyTransforms(
    ItemToModify="((Item=(Type=Bone,Name=\"dyn_shoulderChain_A01_up\"),Transform=(Rotation=(X=0,Y=0,Z=0.0872,W=0.9962),Translation=(X=0.05,Y=-0.05,Z=0),Scale3D=(X=1,Y=1,Z=1))),(Item=(Type=Bone,Name=\"dyn_shoulderChain_B01_up\"),Transform=(Rotation=(X=0,Y=-0.0863,Z=0,W=0.9963),T ...(truncated)"
    WeightMinimum=0
    ExecutePin=@n1.ExecutePin
    Weight=@n3.Result
   )
   > &RigUnit_ModifyTransforms(
    ItemToModify="((Item=(Type=Bone,Name=\"dyn_shoulderChain_A01_dw\"),Transform=(Rotation=(X=-0.0066,Y=0.0696,Z=0.0351,W=0.9969),Translation=(X=0.1,Y=-0.1,Z=-0.1),Scale3D=(X=1,Y=1,Z=1))),(Item=(Type=Bone,Name=\"dyn_shoulderChain_B01_dw\"),Transform=(Rotation=(X=-0.1195,Y=-0.04 ...(truncated)"
    WeightMinimum=0
    Weight=@n3.Result
   )
   n4 = &RigVMFunction_AnimEvalRichCurve(
    Curve=(EditorCurveData=(Keys=((),(Time=0.5,Value=1),(Time=0.8))))
    SourceMaximum=2
    SourceMinimum=0
    TargetMinimum=0
    Value=&RigVMFunction_MathDoubleRemap(
     SourceMaximum=75
     SourceMinimum=45
     TargetMaximum=1
     TargetMinimum=0
     bClamp=false
     Value=&RigUnit_GetRelativeTransformForItem(Child=(Type=Bone,Name="pelvis"), Parent=(Type=Bone,Name="root"), bChildInitial=False, bParentInitial=False).RelativeTransform.Translation.Z
    ).Result
   )
   n5 = &RigVMFunction_MathDoubleMul(A=@n2.Result, B=@n4.Result)
   Remap = &RigVMFunction_MathDoubleRemap(SourceMaximum=0.8, SourceMinimum=0, TargetMaximum=1, TargetMinimum=0, bClamp=False, Value=@n5.Result)
   > &RigUnit_ModifyTransforms(
    ItemToModify="((Item=(Type=Bone,Name=\"dyn_shoulderChain_A01_up\"),Transform=(Rotation=(X=0,Y=0,Z=0.0872,W=0.9962),Translation=(X=0,Y=0,Z=0),Scale3D=(X=1,Y=1,Z=1))),(Item=(Type=Bone,Name=\"dyn_shoulderChain_B01_up\"),Transform=(Rotation=(X=0,Y=0,Z=-0.0872,W=0.9962),Translat ...(truncated)"
    WeightMinimum=0
    Weight=@Remap.Result
   )
   > &RigUnit_ModifyTransforms(
    ItemToModify="((Item=(Type=Bone,Name=\"dyn_shoulderChain_A01_dw\"),Transform=(Rotation=(X=0,Y=0,Z=0.0872,W=0.9962),Translation=(X=0,Y=0,Z=0),Scale3D=(X=1,Y=1,Z=1))),(Item=(Type=Bone,Name=\"dyn_shoulderChain_B01_dw\"),Transform=(Rotation=(X=0,Y=0,Z=-0.0872,W=0.9962),Translat ...(truncated)"
    WeightMinimum=0
    Weight=@Remap.Result
   )
   > DeltaRotation_1 = RigVMCollapseNode(
    Rotation Order=ZYX
    child=upperarm_twist_01_r
    debugRotation=false
    max=152
    min=105
    parent=upperarm_r_null
    parentType=Null
    premult=1
    primAxis=2
    twistAxis=(X=1,Y=0,Z=0)
   )
   Multiply = &RigVMFunction_MathDoubleMul(A=@DeltaRotation_1.outputWeight, B=@n5.Result)
   n6 = &RigVMFunction_MathDoubleRemap(SourceMaximum=0.8, SourceMinimum=0, TargetMaximum=1, TargetMinimum=0, bClamp=False, Value=@Multiply.Result)
   > &RigUnit_ModifyTransforms(
    ItemToModify="((Item=(Type=Bone,Name=\"dyn_shoulderChain_A01_up\"),Transform=(Rotation=(X=0.3984,Y=-0.04,Z=0.0174,W=0.9162),Translation=(X=0,Y=-0.1,Z=0),Scale3D=(X=1,Y=1,Z=1))),(Item=(Type=Bone,Name=\"dyn_shoulderChain_B01_up\"),Transform=(Rotation=(X=-0.2676,Y=-0.2057,Z=-0 ...(truncated)"
    WeightMinimum=0
    Weight=@n6.Result
   )
   > n7 = &RigUnit_ModifyTransforms(
    ItemToModify="((Item=(Type=Bone,Name=\"dyn_shoulderChain_A01_dw\"),Transform=(Rotation=(X=-0.1626,Y=0.0974,Z=-0.0013,W=0.9819),Translation=(X=0.1,Y=0.3,Z=0),Scale3D=(X=1,Y=1,Z=1))),(Item=(Type=Bone,Name=\"dyn_shoulderChain_B01_dw\"),Transform=(Rotation=(X=-0.0057,Y=-0.0432, ...(truncated)"
    WeightMinimum=0
    Weight=@n6.Result
   )
   RigVMFunctionReturnNode(ExecutePin=@n7.ExecutePin)
   DeltaRotation_2 = RigVMCollapseNode(
    ExecuteContext=()
    Rotation Order=ZYX
    child=upperarm_twist_01_r
    debugRotation=false
    max=152
    min=105
    parent=upperarm_r_null
    parentType=Null
    premult=1
    primAxis=2
    twistAxis=(X=1,Y=0,Z=0)
   )
   &RigVMFunction_MathDoubleMul(A=@DeltaRotation_2.outputWeight, B=@n5.Result)
  ]}
 CollapseNode_ContainedGraph {nodes [
   n1 = RigVMFunctionEntryNode()
   &RigUnit_ParentConstraint(
    AdvancedSettings=(RotationOrderForFilter=XZY)
    Child=(Type=Bone,Name="base_collar_1_l")
    Parents="((Item=(Type=Bone,Name=\"spine_03\"),Weight=0.0181),(Item=(Type=Bone,Name=\"neck_01\"),Weight=0.0191),(Item=(Type=Bone,Name=\"deform_pec_l\"),Weight=0.0239),(Item=(Type=Bone,Name=\"spine_04\"),Weight=0.0514),(Item=(Type=Bone,Name=\"neck_02\"),Weight=0.052),(Item=(Ty ...(truncated)"
    Weight=1
    ExecutePin=@n1.ExecutePin
   )
   > &RigUnit_ParentConstraint(
    AdvancedSettings=(RotationOrderForFilter=XZY)
    Child=(Type=Bone,Name="base_collar_2_l")
    Parents=((Item=(Type=Bone,Name="neck_01"),Weight=0.0501),(Item=(Type=Bone,Name="neck_02"),Weight=0.1253),(Item=(Type=Bone,Name="clavicle_l"),Weight=0.1551),(Item=(Type=Bone,Name="spine_05"),Weight=0.6694))
    Weight=1
   )
   > &RigUnit_ParentConstraint(
    AdvancedSettings=(RotationOrderForFilter=XZY)
    Child=(Type=Bone,Name="base_collar_3_l")
    Parents=((Item=(Type=Bone,Name="neck_01"),Weight=0.073),(Item=(Type=Bone,Name="clavicle_l"),Weight=0.0804),(Item=(Type=Bone,Name="neck_02"),Weight=0.1702),(Item=(Type=Bone,Name="spine_05"),Weight=0.6763))
    Weight=1
   )
   > &RigUnit_ParentConstraint(
    AdvancedSettings=(RotationOrderForFilter=XZY)
    Child=(Type=Bone,Name="base_collar_1_c")
    Parents=((Item=(Type=Bone,Name="neck_01"),Weight=0.0476),(Item=(Type=Bone,Name="neck_02"),Weight=0.1021),(Item=(Type=Bone,Name="spine_05"),Weight=0.8503))
    Weight=1
   )
   > &RigUnit_ParentConstraint(
    AdvancedSettings=(RotationOrderForFilter=XZY)
    Child=(Type=Bone,Name="base_collar_1_r")
    Parents="((Item=(Type=Bone,Name=\"neck_01\"),Weight=0.0168),(Item=(Type=Bone,Name=\"spine_03\"),Weight=0.0188),(Item=(Type=Bone,Name=\"deform_pec_r\"),Weight=0.0248),(Item=(Type=Bone,Name=\"neck_02\"),Weight=0.0474),(Item=(Type=Bone,Name=\"spine_04\"),Weight=0.0523),(Item=(T ...(truncated)"
    Weight=1
   )
   > &RigUnit_ParentConstraint(
    AdvancedSettings=(RotationOrderForFilter=XZY)
    Child=(Type=Bone,Name="base_collar_2_r")
    Parents=((Item=(Type=Bone,Name="neck_01"),Weight=0.0495),(Item=(Type=Bone,Name="neck_02"),Weight=0.124),(Item=(Type=Bone,Name="clavicle_r"),Weight=0.1559),(Item=(Type=Bone,Name="spine_05"),Weight=0.6706))
    Weight=1
   )
   > n2 = &RigUnit_ParentConstraint(
    AdvancedSettings=(RotationOrderForFilter=XZY)
    Child=(Type=Bone,Name="base_collar_3_r")
    Parents=((Item=(Type=Bone,Name="neck_01"),Weight=0.073),(Item=(Type=Bone,Name="clavicle_r"),Weight=0.0805),(Item=(Type=Bone,Name="neck_02"),Weight=0.1702),(Item=(Type=Bone,Name="spine_05"),Weight=0.6763))
    Weight=1
   )
   RigVMFunctionReturnNode(ExecutePin=@n2.ExecutePin)
  ]}
 DeltaRotation_ContainedGraph {nodes [
   n1 = RigVMFunctionEntryNode()
   RigVMFunction_MathBoolToInteger = &RigVMFunction_MathBoolToInteger(Value=@n1.onlyTwist)
   RigVMFunction_MathQuaternionSwingTwist = &RigVMFunction_MathQuaternionSwingTwist(
    twistAxis=@n1.twistAxis
    Input=&RigUnit_GetRelativeTransformForItem(
     bChildInitial=False
     bParentInitial=False
     child=(Type=Bone,Name="None")
     parent=(Type=Bone,Name="None")
     child.Name=@n1.child
     parent.Name=@n1.parent
     parent.Type=@n1.parentType
    ).RelativeTransform.Rotation
   )
   RigVMFunction_MathQuaternionToEuler = &RigVMFunction_MathQuaternionToEuler(
    Value=RigVMDispatchNode(
     Values=((),())
     Index=@RigVMFunction_MathBoolToInteger.Result
     Values.0=@RigVMFunction_MathQuaternionSwingTwist.Swing
     Values.1=@RigVMFunction_MathQuaternionSwingTwist.Twist
    ) {template="DISPATCH_RigVMDispatch_SelectInt32(in Index,in Values,out Result)"}.Result
    RotationOrder=@n1.Rotation Order
   )
   > DISPATCH_RigVMDispatch_Print = RigVMDispatchNode(ScreenColor=(R=1,G=1,B=1,A=1), ScreenDuration=0.1, Enabled=@n1.debugRotation, Value=@RigVMFunction_MathQuaternionToEuler.Result) {template="DISPATCH_RigVMDispatch_Print(in Prefix,in Value,in Enabled,in ScreenDuration,in ScreenColor)"}
   Clamp = &RigVMFunction_MathIntClamp(Value=@n1.primAxis)
   DISPATCH_RigVMDispatch_SelectInt32 = RigVMDispatchNode(
    Values=(0,0,0)
    Index=@Clamp.Result
    Values.0=@RigVMFunction_MathQuaternionToEuler.Result.X
    Values.1=@RigVMFunction_MathQuaternionToEuler.Result.Y
    Values.2=@RigVMFunction_MathQuaternionToEuler.Result.Z
   ) {template="DISPATCH_RigVMDispatch_SelectInt32(in Index,in Values,out Result)"}
   Remap = &RigVMFunction_MathDoubleRemap(
    TargetMaximum=1
    TargetMinimum=0
    bClamp=False
    Value=@DISPATCH_RigVMDispatch_SelectInt32.Result
    SourceMaximum=@n1.max
    SourceMinimum=@n1.min
   )
   Multiply = &RigVMFunction_MathDoubleMul(A=&RigVMFunction_MathDoubleClamp(Value=@Remap.Result).Result, B=@n1.premult)
   > RigVMFunctionReturnNode(outputWeight=@Multiply.Result)
  ]}
 New Function_ContainedGraph {nodes [
   RigVMFunctionEntryNode()
   > RigVMFunction_Sequence = RigVMAggregateNode()
   &RigUnit_ParentConstraint(
    AdvancedSettings=(RotationOrderForFilter=XZY)
    Child=(Type=Bone,Name="base_shoulderChainA_up")
    Parents="((Item=(Type=Bone,Name=\"upperarm_r\"),Weight=0.0022),(Item=(Type=Bone,Name=\"spine_05\"),Weight=0.0326),(Item=(Type=Bone,Name=\"spine_03\"),Weight=0.0333),(Item=(Type=Bone,Name=\"upperarm_twist_02_r\"),Weight=0.0615),(Item=(Type=Bone,Name=\"deform_upperarm_bk_r\"), ...(truncated)"
    Weight=1
    ExecutePin=@RigVMFunction_Sequence.A
   )
   > &RigUnit_ParentConstraint(
    AdvancedSettings=(RotationOrderForFilter=XZY)
    Child=(Type=Bone,Name="base_shoulderChainB_up")
    Parents="((Item=(Type=Bone,Name=\"spine_04\"),Weight=0.0124),(Item=(Type=Bone,Name=\"spine_03\"),Weight=0.0189),(Item=(Type=Bone,Name=\"spine_05\"),Weight=0.037),(Item=(Type=Bone,Name=\"upperarm_twist_02_r\"),Weight=0.1029),(Item=(Type=Bone,Name=\"deform_upperarm_fr_r\"),Wei ...(truncated)"
    Weight=1
   )
   > &RigUnit_ParentConstraint(
    AdvancedSettings=(RotationOrderForFilter=XZY)
    Child=(Type=Bone,Name="base_shoulderChainA_dw")
    Parents="((Item=(Type=Bone,Name=\"spine_04\"),Weight=0.0109),(Item=(Type=Bone,Name=\"spine_05\"),Weight=0.0322),(Item=(Type=Bone,Name=\"spine_03\"),Weight=0.0418),(Item=(Type=Bone,Name=\"upperarm_twist_02_r\"),Weight=0.0594),(Item=(Type=Bone,Name=\"deform_upperarm_up_r\"),We ...(truncated)"
    Weight=1
   )
   > &RigUnit_ParentConstraint(
    AdvancedSettings=(RotationOrderForFilter=XZY)
    Child=(Type=Bone,Name="base_shoulderChainB_dw")
    Parents="((Item=(Type=Bone,Name=\"spine_03\"),Weight=0.0178),(Item=(Type=Bone,Name=\"spine_04\"),Weight=0.0255),(Item=(Type=Bone,Name=\"spine_05\"),Weight=0.0343),(Item=(Type=Bone,Name=\"upperarm_twist_02_r\"),Weight=0.1174),(Item=(Type=Bone,Name=\"deform_upperarm_up_r\"),We ...(truncated)"
    Weight=1
   )
   > &RigUnit_AimConstraintLocalSpaceOffset(
    AdvancedSettings=(DebugSettings=(bEnabled=false,Scale=10,WorldOffset=(Rotation=(X=0,Y=0,Z=0,W=1),Translation=(X=0,Y=0,Z=0),Scale3D=(X=1,Y=1,Z=1))),RotationOrderForFilter=XZY)
    Child=(Type=Bone,Name="dyn_shoulderChain_B01_dw")
    Filter=(bX=false,bY=True,bZ=false)
    Parents=((Item=(Type=Bone,Name="upperarm_twist_02_r"),Weight=1))
    Weight=0.65
    WorldUp=(Target=(X=0,Y=0,Z=1),Kind=Location,Space=(Type=Bone,Name="upperarm_r"))
   )
   > &RigUnit_AimConstraintLocalSpaceOffset(
    Child=(Type=Bone,Name="dyn_shoulderChain_A01_dw")
    Filter=(bX=false,bY=True,bZ=false)
    Parents=((Item=(Type=Bone,Name="upperarm_twist_02_r"),Weight=1))
    Weight=0.65
    WorldUp=(Target=(X=0,Y=0,Z=1),Kind=Location,Space=(Type=Bone,Name="upperarm_r"))
   )
   Get_Animation_Attribute_1 = RigVMDispatchNode(BoneName=root, Default=0, Name=deformRig_r_upperarm_fr_alpha) {template="DISPATCH_RigDispatch_GetAnimAttribute(in Name,in BoneName,in Default,out Value,out Found)"}
   &RigUnit_ModifyTransforms(
    ItemToModify="((Item=(Type=Bone,Name=\"dyn_shoulderChain_A01_up\"),Transform=(Rotation=(X=-0.0197,Y=0.1725,Z=0.1115,W=0.9785),Translation=(X=0,Y=0,Z=0),Scale3D=(X=1,Y=1,Z=1))),(Item=(Type=Bone,Name=\"dyn_shoulderChain_B01_up\"),Transform=(Rotation=(X=-0.0053,Y=-0.0433,Z=-0. ...(truncated)"
    WeightMinimum=0
    Weight=@Get_Animation_Attribute_1.Value
    ExecutePin=@RigVMFunction_Sequence.B
   )
   > &RigUnit_ModifyTransforms(
    ItemToModify="((Item=(Type=Bone,Name=\"dyn_shoulderChain_A01_dw\"),Transform=(Rotation=(X=0.0018,Y=0.2079,Z=-0.0085,W=0.9781),Translation=(X=0.1,Y=-0.05,Z=0),Scale3D=(X=1,Y=1,Z=1))),(Item=(Type=Bone,Name=\"dyn_shoulderChain_B01_dw\"),Transform=(Rotation=(X=-0.1946,Y=-0.1173 ...(truncated)"
    WeightMinimum=0
    Weight=@Get_Animation_Attribute_1.Value
   )
   > n1 = &RigUnit_SphericalPoseReader(
    ActiveRegionScaleFactors=(PositiveWidth=1,NegativeWidth=1,PositiveHeight=0.75,NegativeHeight=1)
    ActiveRegionSize=0.1
    Debug=(bDrawDebug=false,bDraw2D=false,bDrawLocalAxes=False,DebugScale=25,DebugSegments=20,DebugThickness=0.25)
    DriverAxis=(X=-1,Y=0,Z=0)
    DriverItem=(Type=Bone,Name="upperarm_r")
    FalloffRegionScaleFactors=(PositiveWidth=1.5,NegativeWidth=1,PositiveHeight=1,NegativeHeight=1)
    FalloffSize=0.2
    FlipHeightScaling=False
    FlipWidthScaling=False
    OptionalParentItem=(Type=Bone,Name="None")
    RotationOffset=(X=-140,Y=45,Z=0)
   )
   > &RigUnit_ModifyTransforms(
    ItemToModify="((Item=(Type=Bone,Name=\"dyn_shoulderChain_A01_up\"),Transform=(Rotation=(X=-0.3,Y=0.3087,Z=0.1033,W=0.8967),Translation=(X=0,Y=0,Z=0),Scale3D=(X=1,Y=1,Z=1))),(Item=(Type=Bone,Name=\"dyn_shoulderChain_B01_up\"),Transform=(Rotation=(X=-0.0233,Y=-0.1196,Z=-0.189 ...(truncated)"
    WeightMinimum=0
    Weight=@n1.OutputParam
   )
   > n2 = &RigUnit_ModifyTransforms(
    ItemToModify="((Item=(Type=Bone,Name=\"dyn_shoulderChain_A01_dw\"),Transform=(Rotation=(X=-0.3472,Y=0.1223,Z=0.0457,W=0.9287),Translation=(X=0,Y=0,Z=0),Scale3D=(X=1,Y=1,Z=1))),(Item=(Type=Bone,Name=\"dyn_shoulderChain_B01_dw\"),Transform=(Rotation=(X=-0.2307,Y=-0.2286,Z=-0. ...(truncated)"
    WeightMinimum=0
    Weight=@n1.OutputParam
   )
   Crouching_DEF = RigVMCollapseNode(ExecutePin=@n2.ExecutePin, ADS Up Pistol=@n1.OutputParam)
   RigVMFunctionReturnNode(ExecuteContext=@RigVMFunction_Sequence.D)
   n3 = &RigVMFunction_AlphaInterp(
    InterpSpeedDecreasing=10
    InterpSpeedIncreasing=10
    Scale=1
    bClampResult=true
    Value=&RigVMFunction_MathBoolToFloat(Value=RigVMVariableNode() {variable=isHanggliding}.Value).Result
   )
   &RigUnit_ModifyTransforms(
    ItemToModify="((Item=(Type=Bone,Name=\"dyn_shoulderChain_A01_up\"),Transform=(Rotation=(X=0,Y=-0.1305,Z=0,W=0.9914),Translation=(X=0,Y=0,Z=0),Scale3D=(X=1,Y=1,Z=1))),(Item=(Type=Bone,Name=\"dyn_shoulderChain_B01_up\"),Transform=(Rotation=(X=0,Y=-0.1305,Z=0,W=0.9914),Transla ...(truncated)"
    WeightMinimum=0
    Weight=@n3.Result
    ExecutePin=@RigVMFunction_Sequence.C
   )
   > &RigUnit_ModifyTransforms(
    ItemToModify="((Item=(Type=Bone,Name=\"dyn_shoulderChain_A01_dw\"),Transform=(Rotation=(X=0,Y=-0.1305,Z=0,W=0.9914),Translation=(X=0,Y=0,Z=0),Scale3D=(X=1,Y=1,Z=1))),(Item=(Type=Bone,Name=\"dyn_shoulderChain_B01_dw\"),Transform=(Rotation=(X=0,Y=-0.1392,Z=0,W=0.9903),Transla ...(truncated)"
    WeightMinimum=0
    Weight=@n3.Result
   )
   n4 = &RigVMFunction_AlphaInterp(
    InterpSpeedDecreasing=10
    InterpSpeedIncreasing=10
    Scale=1
    bClampResult=true
    Value=&RigVMFunction_MathBoolToFloat(Value=RigVMVariableNode() {variable=isSkyDivingFloat}.Value).Result
   )
   > &RigUnit_ModifyTransforms(
    ItemToModify="((Item=(Type=Bone,Name=\"dyn_shoulderChain_A01_up\"),Transform=(Rotation=(X=0,Y=-0.2079,Z=0,W=0.9781),Translation=(X=0,Y=0,Z=0),Scale3D=(X=1,Y=1,Z=1))),(Item=(Type=Bone,Name=\"dyn_shoulderChain_B01_up\"),Transform=(Rotation=(X=0,Y=-0.1045,Z=0,W=0.9945),Transla ...(truncated)"
    WeightMinimum=0
    Weight=@n4.Result
   )
   > &RigUnit_ModifyTransforms(
    ItemToModify="((Item=(Type=Bone,Name=\"dyn_shoulderChain_A01_dw\"),Transform=(Rotation=(X=0,Y=-0.2079,Z=0,W=0.9781),Translation=(X=0,Y=0,Z=0),Scale3D=(X=1,Y=1,Z=1))),(Item=(Type=Bone,Name=\"dyn_shoulderChain_B01_dw\"),Transform=(Rotation=(X=0,Y=-0.1045,Z=0,W=0.9945),Transla ...(truncated)"
    WeightMinimum=0
    Weight=@n4.Result
   )
   n5 = &RigVMFunction_AlphaInterp(
    InterpSpeedDecreasing=10
    InterpSpeedIncreasing=10
    Scale=1
    bClampResult=true
    Value=&RigVMFunction_MathBoolToFloat(Value=RigVMVariableNode() {variable=isEmoting}.Value).Result
   )
   > &RigUnit_ModifyTransforms(
    ItemToModify="((Item=(Type=Bone,Name=\"dyn_shoulderChain_A01_up\"),Transform=(Rotation=(X=0,Y=-0.0785,Z=0,W=0.9969),Translation=(X=0,Y=0,Z=0),Scale3D=(X=1,Y=1,Z=1))),(Item=(Type=Bone,Name=\"dyn_shoulderChain_B01_up\"),Transform=(Rotation=(X=0,Y=-0.0785,Z=0,W=0.9969),Transla ...(truncated)"
    WeightMinimum=0
    Weight=@n5.Result
   )
   > &RigUnit_ModifyTransforms(
    ItemToModify="((Item=(Type=Bone,Name=\"dyn_shoulderChain_A01_dw\"),Transform=(Rotation=(X=0,Y=-0.0872,Z=0,W=0.9962),Translation=(X=0,Y=0,Z=0),Scale3D=(X=1,Y=1,Z=1))),(Item=(Type=Bone,Name=\"dyn_shoulderChain_B01_dw\"),Transform=(Rotation=(X=0,Y=-0.0872,Z=0,W=0.9962),Transla ...(truncated)"
    WeightMinimum=0
    Weight=@n5.Result
   )
   n6 = &RigVMFunction_AlphaInterp(
    InterpSpeedDecreasing=10
    InterpSpeedIncreasing=10
    Scale=1
    bClampResult=true
    Value=&RigVMFunction_MathBoolToFloat(Value=RigVMVariableNode() {variable=isSwimSprinting}.Value).Result
   )
   > &RigUnit_ModifyTransforms(
    ItemToModify="((Item=(Type=Bone,Name=\"dyn_shoulderChain_A01_up\"),Transform=(Rotation=(X=0,Y=-0.1045,Z=0,W=0.9945),Translation=(X=0,Y=0,Z=0),Scale3D=(X=1,Y=1,Z=1))),(Item=(Type=Bone,Name=\"dyn_shoulderChain_B01_up\"),Transform=(Rotation=(X=0,Y=-0.1045,Z=0,W=0.9945),Transla ...(truncated)"
    WeightMinimum=0
    Weight=@n6.Result
   )
   > &RigUnit_ModifyTransforms(
    ItemToModify="((Item=(Type=Bone,Name=\"dyn_shoulderChain_A01_dw\"),Transform=(Rotation=(X=0,Y=-0.1045,Z=0,W=0.9945),Translation=(X=0,Y=0,Z=0),Scale3D=(X=1,Y=1,Z=1))),(Item=(Type=Bone,Name=\"dyn_shoulderChain_B01_dw\"),Transform=(Rotation=(X=0,Y=-0.2079,Z=0,W=0.9781),Transla ...(truncated)"
    WeightMinimum=0
    Weight=@n6.Result
   )
  ]}
 New Function_ContainedGraph {nodes [
   n1 = RigVMFunctionEntryNode()
   RigVMFunction_MathBoolToInteger = &RigVMFunction_MathBoolToInteger(Value=@n1.onlyTwist)
   RigVMFunction_MathQuaternionSwingTwist = &RigVMFunction_MathQuaternionSwingTwist(
    TwistAxis=@n1.twistAxis
    Input=&RigUnit_GetRelativeTransformForItem(
     Child=(Type=Bone,Name="None")
     Parent=(Type=Bone,Name="None")
     bChildInitial=False
     bParentInitial=False
     Child.Name=@n1.child
     Parent.Name=@n1.parent
     Parent.Type=@n1.parentType
    ).RelativeTransform.Rotation
   )
   RigVMFunction_MathQuaternionToEuler = &RigVMFunction_MathQuaternionToEuler(
    Value=RigVMDispatchNode(
     Values=((),())
     Index=@RigVMFunction_MathBoolToInteger.Result
     Values.0=@RigVMFunction_MathQuaternionSwingTwist.Swing
     Values.1=@RigVMFunction_MathQuaternionSwingTwist.Twist
    ) {template="DISPATCH_RigVMDispatch_SelectInt32(in Index,in Values,out Result)"}.Result
    RotationOrder=@n1.Rotation Order
   )
   > DISPATCH_RigVMDispatch_Print = RigVMDispatchNode(ScreenColor=(R=1,G=1,B=1,A=1), ScreenDuration=1.5, Enabled=@n1.debugRotation, Value=@RigVMFunction_MathQuaternionToEuler.Result) {template="DISPATCH_RigVMDispatch_Print(in Prefix,in Value,in Enabled,in ScreenDuration,in ScreenColor)"}
   Clamp = &RigVMFunction_MathIntClamp(Value=@n1.primAxis)
   DISPATCH_RigVMDispatch_SelectInt32 = RigVMDispatchNode(
    Values=(0,0,0)
    Index=@Clamp.Result
    Values.0=@RigVMFunction_MathQuaternionToEuler.Result.X
    Values.1=@RigVMFunction_MathQuaternionToEuler.Result.Y
    Values.2=@RigVMFunction_MathQuaternionToEuler.Result.Z
   ) {template="DISPATCH_RigVMDispatch_SelectInt32(in Index,in Values,out Result)"}
   Remap = &RigVMFunction_MathDoubleRemap(
    TargetMaximum=1
    TargetMinimum=0
    bClamp=False
    Value=@DISPATCH_RigVMDispatch_SelectInt32.Result
    SourceMaximum=@n1.max
    SourceMinimum=@n1.min
   )
   Multiply = &RigVMFunction_MathDoubleMul(A=&RigVMFunction_MathDoubleClamp(Value=@Remap.Result).Result, B=@n1.premult)
   > RigVMFunctionReturnNode(outputWeight=@Multiply.Result)
  ]}
 New Function_ContainedGraph {nodes [
   n1 = RigVMFunctionEntryNode()
   RigVMFunction_MathBoolToInteger = &RigVMFunction_MathBoolToInteger(Value=@n1.onlyTwist)
   RigVMFunction_MathQuaternionSwingTwist = &RigVMFunction_MathQuaternionSwingTwist(
    TwistAxis=@n1.twistAxis
    Input=&RigUnit_GetRelativeTransformForItem(
     Child=(Type=Bone,Name="None")
     Parent=(Type=Bone,Name="None")
     bChildInitial=False
     bParentInitial=False
     Child.Name=@n1.child
     Parent.Name=@n1.parent
     Parent.Type=@n1.parentType
    ).RelativeTransform.Rotation
   )
   RigVMFunction_MathQuaternionToEuler = &RigVMFunction_MathQuaternionToEuler(
    Value=RigVMDispatchNode(
     Values=((),())
     Index=@RigVMFunction_MathBoolToInteger.Result
     Values.0=@RigVMFunction_MathQuaternionSwingTwist.Swing
     Values.1=@RigVMFunction_MathQuaternionSwingTwist.Twist
    ) {template="DISPATCH_RigVMDispatch_SelectInt32(in Index,in Values,out Result)"}.Result
    RotationOrder=@n1.Rotation Order
   )
   > DISPATCH_RigVMDispatch_Print = RigVMDispatchNode(ScreenColor=(R=1,G=1,B=1,A=1), ScreenDuration=1.5, Enabled=@n1.debugRotation, Value=@RigVMFunction_MathQuaternionToEuler.Result) {template="DISPATCH_RigVMDispatch_Print(in Prefix,in Value,in Enabled,in ScreenDuration,in ScreenColor)"}
   Clamp = &RigVMFunction_MathIntClamp(Value=@n1.primAxis)
   DISPATCH_RigVMDispatch_SelectInt32 = RigVMDispatchNode(
    Values=(0,0,0)
    Index=@Clamp.Result
    Values.0=@RigVMFunction_MathQuaternionToEuler.Result.X
    Values.1=@RigVMFunction_MathQuaternionToEuler.Result.Y
    Values.2=@RigVMFunction_MathQuaternionToEuler.Result.Z
   ) {template="DISPATCH_RigVMDispatch_SelectInt32(in Index,in Values,out Result)"}
   Remap = &RigVMFunction_MathDoubleRemap(
    TargetMaximum=1
    TargetMinimum=0
    bClamp=False
    Value=@DISPATCH_RigVMDispatch_SelectInt32.Result
    SourceMaximum=@n1.max
    SourceMinimum=@n1.min
   )
   Multiply = &RigVMFunction_MathDoubleMul(A=&RigVMFunction_MathDoubleClamp(Value=@Remap.Result).Result, B=@n1.premult)
   > RigVMFunctionReturnNode(outputWeight=@Multiply.Result)
  ]}
 New Function_ContainedGraph {nodes [
   n1 = RigVMFunctionEntryNode()
   RigVMFunction_ControlFlowBranch = &RigVMFunction_ControlFlowBranch(Condition=@n1.Debug, ExecuteContext=@n1.ExecuteContext)
   RigVMFunctionReturnNode(ExecuteContext=@RigVMFunction_ControlFlowBranch.Completed)
   DISPATCH_RigVMDispatch_ArrayIterator = RigVMDispatchNode(
    Array=/Script/ControlRig.RigUnit_CollectionGetAll(TypeToSearch=Control) {is=struct RigVMUnitNode}.Items
    ExecuteContext=@RigVMFunction_ControlFlowBranch.True
   ) {template="DISPATCH_RigVMDispatch_ArrayIterator(in Array,out Element,out Index,out Count,out Ratio)"}
   Replace = /Script/RigVM.RigVMFunction_NameReplace(Old=_ctrl, Name=@DISPATCH_RigVMDispatch_ArrayIterator.Element.Name) {is=struct RigVMUnitNode, template="Replace::Execute(in Name,in Old,in New,out Result)"}
   > &RigUnit_SetTransform(
    Item=(Type=Bone,Name="spine_01")
    Weight=1
    bInitial=False
    Item.Name=@Replace.Result
    Value=&RigUnit_GetTransform(Item=(Type=Control,Name="spine_01_ctrl"), bInitial=False, Item.Name=@DISPATCH_RigVMDispatch_ArrayIterator.Element.Name).Transform
   )
  ]}
 New Function_ContainedGraph {nodes [
   RigVMFunctionEntryNode()
   > DeltaRotation = RigVMFunctionReferenceNode(
    Rotation Order=XYZ
    child=thigh_r
    debugRotation=false
    max=-90
    parent=thigh_r_null
    parentType=Null
    premult=1
    primAxis=2
    twistAxis=(X=1,Y=0,Z=0)
   )
   > &RigUnit_ModifyTransforms(
    ItemToModify=((Item=(Type=Bone,Name="base_beltChain"),Transform=(Rotation=(X=0.1357,Y=0.2073,Z=0.7196,W=0.6487),Translation=(X=-2.239,Y=0.5463,Z=0.4442),Scale3D=(X=1,Y=1,Z=1))))
    WeightMinimum=0
    Weight=@DeltaRotation.outputWeight
   )
   > RigVMFunctionReturnNode()
  ]}
 New Function_ContainedGraph {nodes [
   RigVMFunctionEntryNode()
   > RigVMFunction_Sequence = RigVMAggregateNode()
   Constraints = RigVMCollapseNode(ExecutePin=@RigVMFunction_Sequence.A)
   DeltaRotation_4 = RigVMFunctionReferenceNode(
    Rotation Order=ZYX
    child=head
    debugRotation=false
    max=-75
    min=-35
    parent=head_null
    parentType=Null
    premult=1
    primAxis=2
    twistAxis=(X=1,Y=0,Z=0)
    ExecuteContext=@Constraints.ExecutePin
   )
   > &RigUnit_ModifyTransforms(
    ItemToModify=((Item=(Type=Bone,Name="dyn_collar_1_c"),Transform=(Rotation=(X=0,Y=0.1607,Z=0,W=-0.987),Translation=(X=0,Y=0,Z=0),Scale3D=(X=1,Y=1,Z=1))))
    WeightMinimum=0
    Weight=@DeltaRotation_4.outputWeight
   )
   > DeltaRotation_5 = RigVMFunctionReferenceNode(
    Rotation Order=ZYX
    child=head
    debugRotation=false
    max=-95
    min=-60
    parent=head_null
    parentType=Null
    premult=1
    primAxis=2
    twistAxis=(X=1,Y=0,Z=0)
   )
   > &RigUnit_ModifyTransforms(
    ItemToModify="((Item=(Type=Bone,Name=\"dyn_collar_1_c\"),Transform=(Rotation=(X=0,Y=-0.4226,Z=0,W=0.9063),Translation=(X=0,Y=0,Z=0),Scale3D=(X=1,Y=1,Z=1))),(Item=(Type=Bone,Name=\"dyn_collar_3_l\"),Transform=(Rotation=(X=0.1893,Y=-0.2393,Z=0.1277,W=0.9437),Translation=(X=0, ...(truncated)"
    WeightMinimum=0
    Weight=@DeltaRotation_5.outputWeight
   )
   > n1 = &RigUnit_SphericalPoseReader(
    ActiveRegionScaleFactors=(PositiveWidth=0.8,NegativeWidth=0.8,PositiveHeight=0.8,NegativeHeight=0.8)
    ActiveRegionSize=0.1
    Debug=(bDrawDebug=false,bDraw2D=False,bDrawLocalAxes=False,DebugScale=25,DebugSegments=20,DebugThickness=0.25)
    DriverAxis=(X=1,Y=0,Z=0)
    DriverItem=(Type=Bone,Name="upperarm_l")
    FalloffRegionScaleFactors=(PositiveWidth=1,NegativeWidth=1,PositiveHeight=1,NegativeHeight=1)
    FalloffSize=0.2
    FlipHeightScaling=False
    FlipWidthScaling=False
    OptionalParentItem=(Type=Bone,Name="None")
    RotationOffset=(X=70,Y=0,Z=-30)
   )
   > &RigUnit_ModifyTransforms(
    ItemToModify="((Item=(Type=Bone,Name=\"dyn_collar_1_l\"),Transform=(Rotation=(X=0,Y=0.1736,Z=0,W=0.9848),Translation=(X=0,Y=0,Z=0),Scale3D=(X=1,Y=1,Z=1))),(Item=(Type=Bone,Name=\"dyn_collar_2_l\"),Transform=(Rotation=(X=0,Y=0.2588,Z=0,W=0.9659),Translation=(X=0,Y=0,Z=0),Sca ...(truncated)"
    WeightMinimum=0
    Weight=@n1.OutputParam
   )
   n2 = &RigVMFunction_AlphaInterp(
    InterpSpeedDecreasing=10
    InterpSpeedIncreasing=10
    Scale=1
    bClampResult=true
    Value=&RigVMFunction_MathBoolToFloat(Value=RigVMVariableNode() {variable=isHanggliding}.Value).Result
   )
   &RigUnit_ModifyTransforms(
    ItemToModify="((Item=(Type=Bone,Name=\"dyn_collar_2_r\"),Transform=(Rotation=(X=0,Y=0.3827,Z=0,W=0.9239),Translation=(X=0,Y=0,Z=0),Scale3D=(X=1,Y=1,Z=1))),(Item=(Type=Bone,Name=\"dyn_collar_3_r\"),Transform=(Rotation=(X=0,Y=0.2588,Z=0,W=0.9659),Translation=(X=0,Y=0,Z=0),Sca ...(truncated)"
    WeightMinimum=0
    Weight=@n2.Result
    ExecutePin=@RigVMFunction_Sequence.B
   )
   RigVMFunctionReturnNode(ExecuteContext=@RigVMFunction_Sequence.C)
  ]}
 RigVMFunctionLibrary {nodes [
   Controls = RigVMCollapseNode(Debug=False, ExecuteContext=())
   DeltaRotation = RigVMCollapseNode(ExecuteContext=(), Rotation Order=ZYX, debugRotation=false, premult=1, twistAxis=(X=1,Y=0,Z=0))
   FluffAdjust = RigVMCollapseNode(ExecuteContext=())
   HipChain = RigVMCollapseNode(ExecuteContext=())
   ShoulderChains = RigVMCollapseNode(ExecuteContext=())
  ]}
 RigVMFunction_Sequence_1_ContainedGraph {nodes [
   n1 = RigVMFunctionEntryNode()
   RigVMFunction_Sequence = &RigVMFunction_Sequence(ExecuteContext=@n1.ExecuteContext)
   n2 = &RigVMFunction_Sequence(ExecuteContext=@RigVMFunction_Sequence.B)
   n3 = &RigVMFunction_Sequence(ExecuteContext=@n2.B)
   RigVMFunctionReturnNode(A=@RigVMFunction_Sequence.A, B=@n2.A, C=@n3.A, D=@n3.B)
  ]}
 RigVMFunction_Sequence_1_ContainedGraph {nodes [
   n1 = RigVMFunctionEntryNode()
   RigVMFunction_Sequence = &RigVMFunction_Sequence(ExecuteContext=@n1.ExecuteContext)
   n2 = &RigVMFunction_Sequence(ExecuteContext=@RigVMFunction_Sequence.B)
   RigVMFunctionReturnNode(A=@RigVMFunction_Sequence.A, B=@n2.A, C=@n2.B)
  ]}
 RigVMModel {nodes [
   RigUnit_BeginExecution = &RigUnit_BeginExecution()
   Controls = RigVMFunctionReferenceNode(Debug=false, ExecuteContext=@RigUnit_BeginExecution.ExecutePin)
   > FN_DeformRig_FNC = RigVMFunctionReferenceNode(
    DeformRig_ Curve="(deformRig_l_upperarm_up_alpha_72_#1=false,deformRig_r_upperarm_up_alpha_73_#2=false,deformRig_l_upperarm_fr_alpha_74_#3=true,deformRig_r_upperarm_fr_alpha_75_CB51B80 ...(truncated)"
    DeformRig_Bool="(upperarm_1_#4=True,elbow_3_#5=True,wrist_7_#6=True,groin_8_#7=True,glute_17_#8=True,knee_9_40C408324CA261 ...(truncated)"
    DeformRig_Debug=(upperarm_debug_18_#4=False,wrist_debug_20_#6=False,lat_debug_26_#9=False,groinGlute_debug_27_#7=False)
    DeformRig_Value="(upperarm_up_18_#10=(Rotation=(X=0.0,Y=-0.0,Z=0.0,W=1.0),Translation=(X=0.0,Y=0.0,Z=8.0),Scale3D=(X=0.70,Y=1.0,Z=1.0)),upperarm_fwd_fr_19_#11=(Rotation=(X=0.0,Y=-0.0,Z=0.0,W=1.0),Translation=(X=0.0, ...(truncated)"
   )
   > ShoulderChains = RigVMFunctionReferenceNode()
   > HipChain = RigVMFunctionReferenceNode()
   > FluffAdjust = RigVMFunctionReferenceNode()
   RigVMFunction_ControlFlowBranch = &RigVMFunction_ControlFlowBranch(ExecuteContext=@FluffAdjust.ExecuteContext, Condition=RigVMVariableNode() {variable=isFrontEnd}.Value)
   n1 = &RigUnit_SphericalPoseReader(
    ActiveRegionScaleFactors=(PositiveWidth=1,NegativeWidth=1,PositiveHeight=1,NegativeHeight=1)
    ActiveRegionSize=0.1
    Debug=(bDrawDebug=True,bDraw2D=false,bDrawLocalAxes=False,DebugScale=25,DebugSegments=20,DebugThickness=0.25)
    DriverAxis=(X=-1,Y=0,Z=0)
    DriverItem=(Type=Bone,Name="upperarm_r")
    FalloffRegionScaleFactors=(PositiveWidth=1,NegativeWidth=1,PositiveHeight=1,NegativeHeight=1)
    FalloffSize=0.2
    FlipHeightScaling=False
    FlipWidthScaling=False
    OptionalParentItem=(Type=Bone,Name="None")
    RotationOffset=(X=0,Y=65,Z=0)
    ExecutePin=@RigVMFunction_ControlFlowBranch.True
   )
   > &RigUnit_ModifyTransforms(
    ItemToModify="((Item=(Type=Bone,Name=\"base_shoulderChainA_dw\"),Transform=(Rotation=(X=0.1315,Y=-0.0202,Z=0.0466,W=0.99),Translation=(X=0,Y=0,Z=0),Scale3D=(X=1,Y=1,Z=1))),(Item=(Type=Bone,Name=\"base_shoulderChainB_dw\"),Transform=(Rotation=(X=-0.2446,Y=0.148,Z=-0.0498,W=0 ...(truncated)"
    WeightMinimum=0
    Weight=@n1.OutputParam
   )
   RigVMVariableNode() {variable=isFrontEnd}
  ]}
]
hierarchy [
 root {loc=<0 0 0>, rot=<0 0 0 1>}
 pelvis {parent=root, loc=<0 -0.7071 93.1191>, rot=<0.0127 0.707 -0.0127 -0.707>}
 spine_01 {parent=pelvis, loc=<6.7494 -0.6388 0>, rot=<0 0 0.034 -0.9994>}
 spine_02 {parent=spine_01, loc=<7.2518 -0.035 0>, rot=<0 0 0.0238 -0.9997>}
 spine_03 {parent=spine_02, loc=<6.0104 0 0>, rot=<0 0 0.0149 -0.9999>}
 spine_04 {parent=spine_03, loc=<8.1024 0 0>, rot=<0 0 -0.0132 -0.9999>}
 spine_05 {parent=spine_04, loc=<10.0146 0 0>, rot=<0 0 -0.0015 -1>}
 neck_01 {parent=spine_05, loc=<11.3138 0.1529 0>, rot=<0 0 -0.2003 -0.9797>}
 neck_02 {parent=neck_01, loc=<4.5627 0 0>, rot=<0 0 0.0504 -0.9987>}
 head {parent=neck_02, loc=<5.3914 0 0>, rot=<0 0 0.115 -0.9934>}
 base_necklace {parent=spine_05, loc=<12.0023 4.8679 0>, rot=<0 0 0.9511 0.3088>}
 dyn_necklace_01 {parent=base_necklace, loc=<0 0 0>, rot=<0 0 0 1>}
 dyn_necklace_02 {parent=dyn_necklace_01, loc=<3.3987 0 0>, rot=<0 0 0 1>}
 dyn_necklace_03 {parent=dyn_necklace_02, loc=<3.3987 0 0>, rot=<0 0 0 1>}
 dyn_necklace_04 {parent=dyn_necklace_03, loc=<3.3987 0 0>, rot=<0 0 0 1>}
 attach_cape {parent=spine_05, loc=<9.017 -11.714 0>, rot=<0.0616 -0.7044 -0.0616 -0.7044>}
 cape {parent=attach_cape, type=Null, loc=<0 0 0>, rot=<0 0 0 1>}
 attach_backpack {parent=spine_05, loc=<-0.123 -13.473 0>, rot=<-0.0093 -0.707 0.0093 -0.707>}
 Backpack_BR {parent=attach_backpack, type=Null, loc=<0 0 0>, rot=<0 0 0 1>}
 clavicle_l {parent=spine_05, loc=<4.3575 -0.1466 -2.3223>, rot=<0.0465 -0.7289 -0.009 -0.6829>}
 upperarm_l {parent=clavicle_l, loc=<15.5111 0 0>, rot=<0.0226 -0.3234 -0.0088 -0.9459>}
 lowerarm_l {parent=upperarm_l, loc=<26.6514 0 0>, rot=<0.0391 -0.0099 -0.2453 -0.9686>}
 hand_l {parent=lowerarm_l, loc=<22.3491 0 0>, rot=<-0.7081 -0.0507 0.0975 -0.6975>}
 thumb_01_l {parent=hand_l, loc=<1.3118 -0.3191 -2.459>, rot=<0.6546 -0.4042 0.0664 -0.6354>}
 thumb_02_l {parent=thumb_01_l, loc=<5.9336 -0.0448 0.0401>, rot=<-0.0037 0.0213 0.298 -0.9543>}
 thumb_03_l {parent=thumb_02_l, loc=<2.2223 0.3682 -0.0907>, rot=<0.0118 -0.013 0.0066 -0.9998>}
 index_metacarpal_l {parent=hand_l, loc=<1.6535 0.0183 -2.96>, rot=<0.0158 -0.0124 -0.0173 -0.9996>}
 index_01_l {parent=index_metacarpal_l, loc=<8.3008 0.0357 -0.0524>, rot=<-0.0044 0.0485 0.0924 -0.9945>}
 index_02_l {parent=index_01_l, loc=<3.3587 -0.3004 0.0452>, rot=<-0.002 0.0027 0.3225 -0.9466>}
 index_03_l {parent=index_02_l, loc=<2.5578 -0.123 -0.0252>, rot=<-0.002 0.0067 0.2558 -0.9667>}
 middle_metacarpal_l {parent=hand_l, loc=<2.1856 -0.1725 -0.5984>, rot=<0.006 0.0168 -0.0329 -0.9993>}
 middle_01_l {parent=middle_metacarpal_l, loc=<7.3275 -0.2447 0.0129>, rot=<-0.0208 0.0042 0.0858 -0.9961>}
 middle_02_l {parent=middle_01_l, loc=<4.0123 0.1401 0.1175>, rot=<-0.0218 0.0258 0.3512 -0.9357>}
 middle_03_l {parent=middle_02_l, loc=<2.6077 0.0534 0.1373>, rot=<0.0082 0.0188 0.1708 -0.9851>}
 ring_metacarpal_l {parent=hand_l, loc=<2.2074 -0.3625 1.3638>, rot=<-0.0284 0.0722 -0.0458 -0.9959>}
 ring_01_l {parent=ring_metacarpal_l, loc=<7.0295 -0.6197 -0.1446>, rot=<-0.0489 -0.0252 0.1892 -0.9804>}
 ring_02_l {parent=ring_01_l, loc=<3.6174 0.4647 -0.1027>, rot=<0 0 0.2806 -0.9598>}
 ring_03_l {parent=ring_02_l, loc=<2.4152 0.114 0.0199>, rot=<0 0 0.1189 -0.9929>}
 pinky_metacarpal_l {parent=hand_l, loc=<1.7616 -0.9753 3.3814>, rot=<-0.0397 0.0814 -0.069 -0.9935>}
 pinky_01_l {parent=pinky_metacarpal_l, loc=<7.1709 -0.4831 -0.1094>, rot=<-0.0654 -0.0035 0.2242 -0.9723>}
 pinky_02_l {parent=pinky_01_l, loc=<2.8274 0.0893 -0.1043>, rot=<0.0037 0.0011 0.2854 -0.9584>}
 pinky_03_l {parent=pinky_02_l, loc=<2.2046 0.0307 -0.0078>, rot=<-0.0002 0.0098 0.1445 -0.9895>}
 deform_wrist_dn_base_l {parent=hand_l, loc=<0 -2.2342 0>, rot=<0 0 0 -1>}
 deform_wrist_dn_l {parent=deform_wrist_dn_base_l, loc=<0 0 0>, rot=<0 0 0 1>}
 deform_wrist_up_base_l {parent=hand_l, loc=<0 2.2682 0>, rot=<0 0 0 -1>}
 deform_wrist_up_l {parent=deform_wrist_up_base_l, loc=<0 0 0>, rot=<0 0 0 1>}
 lowerarm_twist_02_l {parent=lowerarm_l, loc=<9.6199 0 0>, rot=<0 0 0 -1>}
 base_cuff_l {parent=lowerarm_twist_02_l, loc=<6.1044 0 0>, rot=<0 0 0 1>}
 base_cuff_l_01 {parent=base_cuff_l, loc=<0.048 -3.7047 3.344>, rot=<0.1809 -0.915 0.3053 0.1919>}
 dyn_cuff_l_01 {parent=base_cuff_l_01, loc=<0 0 0>, rot=<0 0 0 1>}
 dyn_cuff_l_01_end {parent=dyn_cuff_l_01, loc=<4.29 0 0>, rot=<0 0 0 1>}
 base_cuff_l_02 {parent=base_cuff_l, loc=<-0.9142 -1.8815 -2.9633>, rot=<0.0425 -0.9551 0.2431 -0.1642>}
 dyn_cuff_l_02 {parent=base_cuff_l_02, loc=<0 0 0>, rot=<0 0 0 1>}
 dyn_cuff_l_02_end {parent=dyn_cuff_l_02, loc=<3.7045 0 0>, rot=<0 0 0 1>}
 base_cuff_l_03 {parent=base_cuff_l, loc=<0.2199 3.9763 -1.12>, rot=<-0.1362 -0.9849 0.1025 0.0293>}
 dyn_cuff_l_03 {parent=base_cuff_l_03, loc=<0 0 0>, rot=<0 0 0 1>}
 dyn_cuff_l_03_end {parent=dyn_cuff_l_03, loc=<2.4715 0 0>, rot=<0 0 0 1>}
 base_cuff_l_04 {parent=base_cuff_l, loc=<0.2492 1.3896 4.1093>, rot=<0.2104 -0.5182 0.7961 0.231>}
 dyn_cuff_l_04 {parent=base_cuff_l_04, loc=<0 0 0>, rot=<0 0 0 1>}
 dyn_cuff_l_04_end {parent=dyn_cuff_l_04, loc=<3.4964 0 0>, rot=<0 0 0 1>}
 lowerarm_twist_01_l {parent=lowerarm_l, loc=<17.855 0 0>, rot=<0 0 0 -1>}
 deform_elbow_in_base_l {parent=lowerarm_l, loc=<0.2112 5.259 0.8274>, rot=<0 0 0 -1>}
 deform_elbow_in_l {parent=deform_elbow_in_base_l, loc=<0 0 0>, rot=<0 0 0 1>}
 deform_elbow_out_base_l {parent=lowerarm_l, loc=<0 -4 0>, rot=<0 0 0 -1>}
 deform_elbow_out_l {parent=deform_elbow_out_base_l, loc=<0 0 0>, rot=<0 0 0 1>}
 upperarm_twist_01_l {parent=upperarm_l, loc=<0.0041 0 0>, rot=<0 0 0 -1>}
 deform_upperarm_bk_base_l {parent=upperarm_twist_01_l, loc=<1.0813 -7.2193 -1.9453>, rot=<0 0 0 1>}
 deform_upperarm_bk_l {parent=deform_upperarm_bk_base_l, loc=<0 0 0>, rot=<0 0 0 1>}
 deform_upperarm_fr_base_l {parent=upperarm_twist_01_l, loc=<-0.3691 7.0506 -2.765>, rot=<0 0 0 1>}
 deform_upperarm_fr_l {parent=deform_upperarm_fr_base_l, loc=<0 0 0>, rot=<0 0 0 1>}
 deform_upperarm_up_base_l {parent=upperarm_twist_01_l, loc=<-3.2216 -1.1391 3.3311>, rot=<0 0 0 1>}
 deform_upperarm_up_l {parent=deform_upperarm_up_base_l, loc=<0 0 0>, rot=<0 0 0 1>}
 upperarm_twist_02_l {parent=upperarm_l, loc=<13.7941 0 0>, rot=<0 0 0 -1>}
 clavicle_r {parent=spine_05, loc=<4.3575 -0.1466 2.3223>, rot=<0.7289 0.0465 -0.6829 0.009>}
 upperarm_r {parent=clavicle_r, loc=<-15.5111 0 0>, rot=<0.0226 -0.3234 -0.0088 -0.9459>}
 lowerarm_r {parent=upperarm_r, loc=<-26.6514 0 0>, rot=<0.0391 -0.0099 -0.2453 -0.9686>}
 hand_r {parent=lowerarm_r, loc=<-22.3491 0 0>, rot=<-0.7081 -0.0507 0.0975 -0.6975>}
 thumb_01_r {parent=hand_r, loc=<-1.3118 0.3191 2.459>, rot=<0.6546 -0.4042 0.0664 -0.6354>}
 thumb_02_r {parent=thumb_01_r, loc=<-5.9336 0.0448 -0.0401>, rot=<-0.0037 0.0213 0.298 -0.9543>}
 thumb_03_r {parent=thumb_02_r, loc=<-2.2223 -0.3682 0.0907>, rot=<0.0118 -0.013 0.0066 -0.9998>}
 index_metacarpal_r {parent=hand_r, loc=<-1.6535 -0.0183 2.96>, rot=<0.0158 -0.0124 -0.0173 -0.9996>}
 index_01_r {parent=index_metacarpal_r, loc=<-8.3008 -0.0357 0.0524>, rot=<-0.0044 0.0485 0.0924 -0.9945>}
 index_02_r {parent=index_01_r, loc=<-3.3587 0.3004 -0.0452>, rot=<-0.002 0.0027 0.3225 -0.9466>}
 index_03_r {parent=index_02_r, loc=<-2.5578 0.123 0.0252>, rot=<-0.002 0.0067 0.2558 -0.9667>}
 middle_metacarpal_r {parent=hand_r, loc=<-2.1856 0.1725 0.5984>, rot=<0.006 0.0168 -0.0329 -0.9993>}
 middle_01_r {parent=middle_metacarpal_r, loc=<-7.3275 0.2447 -0.0129>, rot=<-0.0208 0.0042 0.0858 -0.9961>}
 middle_02_r {parent=middle_01_r, loc=<-4.0123 -0.1401 -0.1175>, rot=<-0.0218 0.0258 0.3512 -0.9357>}
 middle_03_r {parent=middle_02_r, loc=<-2.6077 -0.0534 -0.1373>, rot=<0.0082 0.0188 0.1708 -0.9851>}
 ring_metacarpal_r {parent=hand_r, loc=<-2.2074 0.3625 -1.3638>, rot=<-0.0284 0.0722 -0.0458 -0.9959>}
 ring_01_r {parent=ring_metacarpal_r, loc=<-7.0295 0.6197 0.1446>, rot=<-0.0489 -0.0252 0.1892 -0.9804>}
 ring_02_r {parent=ring_01_r, loc=<-3.6174 -0.4647 0.1027>, rot=<0 0 0.2806 -0.9598>}
 ring_03_r {parent=ring_02_r, loc=<-2.4152 -0.114 -0.0199>, rot=<0 0 0.1189 -0.9929>}
 pinky_metacarpal_r {parent=hand_r, loc=<-1.7616 0.9753 -3.3814>, rot=<-0.0397 0.0814 -0.069 -0.9935>}
 pinky_01_r {parent=pinky_metacarpal_r, loc=<-7.1709 0.4831 0.1094>, rot=<-0.0654 -0.0035 0.2242 -0.9723>}
 pinky_02_r {parent=pinky_01_r, loc=<-2.8274 -0.0893 0.1043>, rot=<0.0037 0.0011 0.2854 -0.9584>}
 pinky_03_r {parent=pinky_02_r, loc=<-2.2046 -0.0307 0.0078>, rot=<-0.0002 0.0098 0.1445 -0.9895>}
 deform_wrist_dn_base_r {parent=hand_r, loc=<-0.0004 2.2345 -0.0001>, rot=<0 0 0 -1>}
 deform_wrist_dn_r {parent=deform_wrist_dn_base_r, loc=<0 0 0>, rot=<0 0 0 1>}
 deform_wrist_up_base_r {parent=hand_r, loc=<0.0003 -2.2684 0.0001>, rot=<0 0 0 -1>}
 deform_wrist_up_r {parent=deform_wrist_up_base_r, loc=<0 0 0>, rot=<0 0 0 1>}
 lowerarm_twist_02_r {parent=lowerarm_r, loc=<-9.6199 0 0>, rot=<0 0 0 -1>}
 base_cuff_r {parent=lowerarm_twist_02_r, loc=<-6.1048 0.0001 0.0003>, rot=<0 0 0 -1>}
 base_cuff_r_01 {parent=base_cuff_r, loc=<0.2963 3.1303 -2.9686>, rot=<0.3367 -0.8698 0.2674 0.2419>}
 dyn_cuff_r_01 {parent=base_cuff_r_01, loc=<0 0 0>, rot=<0 0 0 1>}
 dyn_cuff_r_01_end {parent=dyn_cuff_r_01, loc=<-4.2899 0 -0.0001>, rot=<0 0 0 1>}
 base_cuff_r_02 {parent=base_cuff_r, loc=<1.9234 2.0539 3.2768>, rot=<0.0425 -0.9551 0.2431 -0.1642>}
 dyn_cuff_r_02 {parent=base_cuff_r_02, loc=<0 0 0>, rot=<0 0 0 1>}
 dyn_cuff_r_02_end {parent=dyn_cuff_r_02, loc=<-3.7046 0 0.0002>, rot=<0 0 0 1>}
 base_cuff_r_03 {parent=base_cuff_r, loc=<0.9716 -4.3163 1.0825>, rot=<-0.1362 -0.9849 0.1025 0.0293>}
 dyn_cuff_r_03 {parent=base_cuff_r_03, loc=<0 0 0>, rot=<0 0 0 1>}
 dyn_cuff_r_03_end {parent=dyn_cuff_r_03, loc=<-2.3479 -0.0001 0>, rot=<0 0 0 1>}
 base_cuff_r_04 {parent=base_cuff_r, loc=<-0.2487 -1.3898 -4.1099>, rot=<0.2104 -0.5182 0.7961 0.231>}
 dyn_cuff_r_04 {parent=base_cuff_r_04, loc=<0 0 0>, rot=<0 0 0 -1>}
 dyn_cuff_r_04_end {parent=dyn_cuff_r_04, loc=<-3.4959 -0.0001 0>, rot=<0 0 0 -1>}
 lowerarm_twist_01_r {parent=lowerarm_r, loc=<-17.8549 0 0>, rot=<0 0 0 -1>}
 deform_elbow_in_base_r {parent=lowerarm_r, loc=<-0.2111 -5.2591 -0.8275>, rot=<0 0 0 -1>}
 deform_elbow_in_r {parent=deform_elbow_in_base_r, loc=<-0.0001 0 0.0002>, rot=<0 0 0 1>}
 deform_elbow_out_base_r {parent=lowerarm_r, loc=<0 4 -0.0001>, rot=<0 0 0 -1>}
 deform_elbow_out_r {parent=deform_elbow_out_base_r, loc=<0 0 0>, rot=<0 0 0 1>}
 upperarm_twist_01_r {parent=upperarm_r, loc=<-0.0041 0 0>, rot=<0 0 0 -1>}
 deform_upperarm_bk_base_r {parent=upperarm_twist_01_r, loc=<-1.0814 7.2192 1.9455>, rot=<0 0 0 1>}
 deform_upperarm_bk_r {parent=deform_upperarm_bk_base_r, loc=<0 0 0>, rot=<0 0 0 1>}
 deform_upperarm_fr_base_r {parent=upperarm_twist_01_r, loc=<0.3689 -7.0506 2.7652>, rot=<0 0 0 1>}
 deform_upperarm_fr_r {parent=deform_upperarm_fr_base_r, loc=<0 0 0>, rot=<0 0 0 1>}
 deform_upperarm_up_base_r {parent=upperarm_twist_01_r, loc=<3.2214 1.1391 -3.3308>, rot=<0 0 0 1>}
 deform_upperarm_up_r {parent=deform_upperarm_up_base_r, loc=<0 0 0>, rot=<0 0 0 -1>}
 upperarm_twist_02_r {parent=upperarm_r, loc=<-13.7941 0 0>, rot=<0 0 0 -1>}
 base_shoulderChain {parent=upperarm_r, loc=<0 0 0>, rot=<0 0 0 1>}
 base_shoulderChainA_up {parent=base_shoulderChain, loc=<0.7447 7.1057 -3.685>, rot=<0.1884 -0.9652 0.0254 -0.1796>}
 dyn_shoulderChain_A01_up {parent=base_shoulderChainA_up, loc=<0 0 0>, rot=<0 0 0 1>}
 dyn_shoulderChain_A02_up {parent=dyn_shoulderChain_A01_up, loc=<4.7442 -0.7586 0>, rot=<0 0 0.1579 -0.9875>}
 base_shoulderChainB_up {parent=base_shoulderChain, loc=<1.1895 -4.2524 -3.815>, rot=<-0.1362 -0.98 -0.0402 -0.1392>}
 dyn_shoulderChain_B01_up {parent=base_shoulderChainB_up, loc=<0 0 0>, rot=<0 0 0 1>}
 dyn_shoulderChain_B02_up {parent=dyn_shoulderChain_B01_up, loc=<4.7523 0.7336 0>, rot=<0 0 -0.1526 -0.9883>}
 base_shoulderChainA_dw {parent=base_shoulderChain, loc=<-0.19 8.4214 -1.2501>, rot=<0.1625 -0.9655 0.0278 -0.2017>}
 dyn_shoulderChain_A01_dw {parent=base_shoulderChainA_dw, loc=<0 0 0>, rot=<0 0 0 1>}
 dyn_shoulderChain_A02_dw {parent=dyn_shoulderChain_A01_dw, loc=<6.039 0 0>, rot=<0 0 0.2193 -0.9757>}
 base_shoulderChainB_dw {parent=base_shoulderChain, loc=<0.3584 -5.352 -1.5998>, rot=<-0.0804 -0.9829 -0.0298 -0.1633>}
 dyn_shoulderChain_B01_dw {parent=base_shoulderChainB_dw, loc=<0 0 0>, rot=<0 0 0 1>}
 dyn_shoulderChain_B02_dw {parent=dyn_shoulderChain_B01_dw, loc=<6.0371 0 0>, rot=<0 0 -0.2372 -0.9715>}
 upperarm_twist_01_r_null {parent=upperarm_r, type=Null, loc=<-0.0041 0 0>, rot=<0 0 0 -1>}
 upperarm_r_null {parent=clavicle_r, type=Null, loc=<-15.5111 0 0>, rot=<0.0226 -0.3234 -0.0088 -0.9459>}
 clavicle_r_null {parent=spine_05, type=Null, loc=<4.3575 -0.1466 2.3223>, rot=<0.7289 0.0465 -0.6829 0.009>}
 clavicle_l_null {parent=spine_05, type=Null, loc=<4.3575 -0.1466 -2.3223>, rot=<0.0465 -0.7289 -0.009 -0.6829>}
 base_collar_1_l {parent=spine_05, loc=<7.1654 6.7506 -7.4759>, rot=<0.8531 0.2821 -0.4223 0.1197>}
 dyn_collar_1_l {parent=base_collar_1_l, loc=<0 0 0>, rot=<0 0 0 1>}
 base_collar_2_l {parent=spine_05, loc=<11.9912 1.2957 -8.6275>, rot=<0.892 0.2424 -0.357 0.1344>}
 dyn_collar_2_l {parent=base_collar_2_l, loc=<0 0 0>, rot=<0 0 0 1>}
 base_collar_3_l {parent=spine_05, loc=<14.5556 -5.2942 -6.4513>, rot=<0.8641 0.088 -0.2877 0.4036>}
 dyn_collar_3_l {parent=base_collar_3_l, loc=<0 0 0>, rot=<0 0 0 1>}
 base_collar_1_c {parent=spine_05, loc=<13.3007 -7.9903 0>, rot=<-0.6956 0.1268 0.1268 -0.6956>}
 dyn_collar_1_c {parent=base_collar_1_c, loc=<0 0 0>, rot=<0 0 0 1>}
 base_collar_1_r {parent=spine_05, loc=<7.1651 6.7505 7.4759>, rot=<-0.1197 0.4223 -0.2821 -0.8531>}
 dyn_collar_1_r {parent=base_collar_1_r, loc=<0 0 0>, rot=<0 0 0 1>}
 base_collar_2_r {parent=spine_05, loc=<11.9907 1.2956 8.6275>, rot=<-0.1344 0.357 -0.2424 -0.892>}
 dyn_collar_2_r {parent=base_collar_2_r, loc=<0 0 0>, rot=<0 0 0 1>}
 base_collar_3_r {parent=spine_05, loc=<14.5561 -5.2942 6.4513>, rot=<-0.4036 0.2877 -0.088 -0.8641>}
 dyn_collar_3_r {parent=base_collar_3_r, loc=<0 0 0>, rot=<0 0 0 1>}
 head_null {parent=spine_05, type=Null, loc=<20.6573 3.5489 0>, rot=<0 0 -0.036 -0.9994>}
 deform_pec_aim_base_l {parent=spine_04, loc=<7.8732 12.6388 -2.4123>, rot=<0.0102 -0.4602 0.3979 -0.7936>}
 deform_pec_aim_l {parent=deform_pec_aim_base_l, loc=<0 0 0>, rot=<0 0 0 1>}
 deform_pec_base_l {parent=deform_pec_aim_l, loc=<10.766 0 0>, rot=<0 0 0 1>}
 deform_pec_l {parent=deform_pec_base_l, loc=<0.055 0 0>, rot=<0 0 0 1>}
 deform_pec_aim_base_r {parent=spine_04, loc=<7.8727 12.6388 2.4123>, rot=<-0.4602 -0.0102 0.7936 0.3979>}
 deform_pec_aim_r {parent=deform_pec_aim_base_r, loc=<0 0 0>, rot=<0 0 0 1>}
 deform_pec_base_r {parent=deform_pec_aim_r, loc=<-10.7662 -0.0006 -0.0007>, rot=<0 0 0 1>}
 deform_pec_r {parent=deform_pec_base_r, loc=<0 0 0>, rot=<0 0 0 1>}
 spine_05_null {parent=spine_04, type=Null, loc=<10.0146 0 0>, rot=<0 0 -0.0015 -1>}
 base_buckle_l_up {parent=pelvis, loc=<29.1797 8.0441 -10.2125>, rot=<-0.0341 -0.3492 0.8576 0.3762>}
 dyn_buckle_l_01_up {parent=base_buckle_l_up, loc=<0 0 0>, rot=<0 0 0 1>}
 dyn_buckle_l_02_up {parent=dyn_buckle_l_01_up, loc=<3.117 0 0>, rot=<0 0 0 1>}
 base_buckle_l_dw {parent=pelvis, loc=<25.6722 4.7727 -12.6643>, rot=<-0.0896 -0.2227 0.892 0.383>}
 dyn_buckle_l_01_dw {parent=base_buckle_l_dw, loc=<0 0 0>, rot=<0 0 0 1>}
 dyn_buckle_l_02_dw {parent=dyn_buckle_l_01_dw, loc=<3.2499 0 0>, rot=<0 0 0 1>}
 base_buckle_r_up {parent=pelvis, loc=<30.1587 8.6133 9.6959>, rot=<-0.2532 -0.8408 -0.3521 -0.3239>}
 dyn_buckle_r_01_up {parent=base_buckle_r_up, loc=<0 0 0>, rot=<0 0 0 1>}
 dyn_buckle_r_02_up {parent=dyn_buckle_r_01_up, loc=<3.7115 0 0>, rot=<0 0 0 1>}
 base_buckle_r_dw {parent=pelvis, loc=<26.0561 5.3713 12.2992>, rot=<-0.2992 -0.8464 -0.358 -0.2567>}
 dyn_buckle_r_01_dw {parent=base_buckle_r_dw, loc=<0 0 0>, rot=<0 0 0 1>}
 dyn_buckle_r_02_dw {parent=dyn_buckle_r_01_dw, loc=<3.7581 0 0>, rot=<0 0 0 1>}
 thigh_l {parent=pelvis, loc=<-2.5693 -0.0109 -7.7501>, rot=<0.0012 0.0572 0.0155 -0.9982>}
 calf_l {parent=thigh_l, loc=<-39.0484 0 0>, rot=<0.0006 -0.0161 -0.0391 -0.9991>}
 foot_l {parent=calf_l, loc=<-42.561 0.1077 0.4292>, rot=<0.0418 -0.0455 0.0343 -0.9975>}
 ball_l {parent=foot_l, loc=<-6.1599 13.8264 0.6806>, rot=<0.0001 -0.0001 -0.7063 -0.7079>}
 calf_twist_02_l {parent=calf_l, loc=<-18.8064 -0.0967 0.0397>, rot=<0.0237 0.0095 -0.0035 -0.9997>}
 calf_twist_01_l {parent=calf_l, loc=<-35.813 -0.0967 0.0397>, rot=<0.0237 0.0095 -0.0035 -0.9997>}
 deform_knee_in_base_l {parent=calf_l, loc=<1.298 -5.2957 -0.0496>, rot=<0 0 0 -1>}
 deform_knee_in_l {parent=deform_knee_in_base_l, loc=<0 0 0>, rot=<0 0 0 1>}
 deform_knee_out_base_l {parent=calf_l, loc=<0 4 0>, rot=<0 0 0 -1>}
 deform_knee_out_l {parent=deform_knee_out_base_l, loc=<0 0 0>, rot=<0 0 0 1>}
 thigh_twist_01_l {parent=thigh_l, loc=<-21.6303 0 0>, rot=<0.0013 0 0 -1>}
 deform_groin_base_l {parent=thigh_l, loc=<-2.1908 8.1142 -1.7442>, rot=<0 0 0 -1>}
 deform_groin_l {parent=deform_groin_base_l, loc=<0 0 0>, rot=<0 0 0 1>}
 thigh_r {parent=pelvis, loc=<-2.5693 -0.0109 7.7501>, rot=<0.0572 -0.0012 0.9982 0.0155>}
 calf_r {parent=thigh_r, loc=<39.0484 0 0>, rot=<0.0006 -0.0161 -0.0391 -0.9991>}
 foot_r {parent=calf_r, loc=<42.561 -0.1077 -0.4292>, rot=<0.0418 -0.0455 0.0343 -0.9975>}
 ball_r {parent=foot_r, loc=<6.1599 -13.8264 -0.6806>, rot=<0.0001 -0.0001 -0.7063 -0.7079>}
 calf_twist_02_r {parent=calf_r, loc=<18.8064 0.0967 -0.0397>, rot=<0.0237 0.0095 -0.0035 -0.9997>}
 calf_twist_01_r {parent=calf_r, loc=<35.813 0.0967 -0.0397>, rot=<0.0237 0.0095 -0.0035 -0.9997>}
 deform_knee_in_base_r {parent=calf_r, loc=<-1.298 5.2957 0.0496>, rot=<0 0 0 -1>}
 deform_knee_in_r {parent=deform_knee_in_base_r, loc=<0 0 0>, rot=<0 0 0 1>}
 deform_knee_out_base_r {parent=calf_r, loc=<0 -4 0>, rot=<0 0 0 -1>}
 deform_knee_out_r {parent=deform_knee_out_base_r, loc=<0 0 0>, rot=<0 0 0 1>}
 thigh_twist_01_r {parent=thigh_r, loc=<21.6303 0 0>, rot=<0.0013 0 0 -1>}
 deform_groin_base_r {parent=thigh_r, loc=<2.1908 -8.1142 1.7442>, rot=<0 0 0 -1>}
 deform_groin_r {parent=deform_groin_base_r, loc=<0 0 0>, rot=<0 0 0 1>}
 thigh_r_null {parent=pelvis, type=Null, loc=<-2.5693 -0.0109 7.7501>, rot=<0.0572 -0.0012 0.9982 0.0155>}
 deform_glute_base_l {parent=pelvis, loc=<-8.8064 -8.8222 -8.4287>, rot=<0.0275 0.0501 -0.4472 -0.8926>}
 deform_glute_l {parent=deform_glute_base_l, loc=<0 0 0>, rot=<0 0 0 1>}
 deform_glute_base_r {parent=pelvis, loc=<-8.8062 -8.8222 8.4287>, rot=<0.0501 -0.0275 0.8926 -0.4472>}
 deform_glute_r {parent=deform_glute_base_r, loc=<0 0 0>, rot=<0 0 0 1>}
 base_beltChain {parent=pelvis, loc=<1.589 8.1252 6.9545>, rot=<-0.2371 -0.9271 -0.171 -0.2347>}
 dyn_beltChain_01 {parent=base_beltChain, loc=<0 0 0>, rot=<0 0 0 1>}
 dyn_beltChain_02 {parent=dyn_beltChain_01, loc=<1.875 0 0>, rot=<0 0 0 1>}
 dyn_beltChain_con {parent=base_beltChain, loc=<0 0 -4.1876>, rot=<0 0 0 1>}
 ik_hand_root {parent=root, loc=<0 0 0>, rot=<0 0 0 1>}
 ik_hand_gun {parent=ik_hand_root, loc=<-51.9727 4.0883 102.6935>, rot=<0.6084 0.1793 0.4613 -0.6205>}
 pelvis_ctrl {controlType=EULER_TRANSFORM, type=Control, loc=<0 0 0>, rot=<0 0 0 1>}
 spine_01_ctrl {controlType=EULER_TRANSFORM, parent=pelvis_ctrl, type=Control, loc=<0 0 0>, rot=<0 0 0 1>}
 spine_02_ctrl {controlType=EULER_TRANSFORM, parent=spine_01_ctrl, type=Control, loc=<0 0 0>, rot=<0 0 0 1>}
 spine_03_ctrl {controlType=EULER_TRANSFORM, parent=spine_02_ctrl, type=Control, loc=<0 0 0>, rot=<0 0 0 1>}
 spine_04_ctrl {controlType=EULER_TRANSFORM, parent=spine_03_ctrl, type=Control, loc=<0 0 0>, rot=<0 0 0 1>}
 spine_05_ctrl {controlType=EULER_TRANSFORM, parent=spine_04_ctrl, type=Control, loc=<0 0 0>, rot=<0 0 0 1>}
 neck_01_ctrl {controlType=EULER_TRANSFORM, parent=spine_05_ctrl, type=Control, loc=<0 0 0>, rot=<0 0 0 1>}
 head_ctrl {controlType=EULER_TRANSFORM, parent=neck_01_ctrl, type=Control, loc=<0 0 0>, rot=<0 0 0 1>}
 clavicle_l_ctrl {controlType=EULER_TRANSFORM, parent=spine_05_ctrl, type=Control, loc=<0 0 0>, rot=<0 0 0 1>}
 upperarm_l_ctrl {controlType=EULER_TRANSFORM, parent=clavicle_l_ctrl, type=Control, loc=<0 0 0>, rot=<0 0 0 1>}
 clavicle_r_ctrl {controlType=EULER_TRANSFORM, parent=spine_05_ctrl, type=Control, loc=<0 0 0>, rot=<0 0 0 1>}
 upperarm_r_ctrl {controlType=EULER_TRANSFORM, parent=clavicle_r_ctrl, type=Control, loc=<0 0 0>, rot=<0 0 0 1>}
 thigh_l_ctrl {controlType=EULER_TRANSFORM, parent=pelvis_ctrl, type=Control, loc=<0 0 0>, rot=<0 0 0 1>}
 thigh_r_ctrl {controlType=EULER_TRANSFORM, parent=pelvis_ctrl, type=Control, loc=<0 0 0>, rot=<0 0 0 1>}
]
variables [
 Crouching {type=bool, default=False}
 isCrouchWalking {type=double, default=0}
 isEmoting {type=bool, default=False}
 isFrontEnd {type=bool, default=False}
 isHanggliding {type=bool, default=False}
 isHeadUpExt {type=double, default=0}
 isSkyDivingFloat {type=bool, default=False}
 isSwimSprinting {type=bool, default=False}
]
