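# Physics asset for the necklace on the male medium Crown Order body cosmetic, setting up its dangling part for physical motion.
# A chain of four tapering capsules, dyn_necklace_01 to 04, hangs from a tiny base_necklace body, beside larger neck_01 and spine_05 capsules shaped as body volumes.
# Chain joints lock twist and limit swing, with the top joint also locking one swing axis.
# Base-to-neck and neck-to-spine joints leave all angular axes free.
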PhysicsAsset /BRCosmetics/Characters/Player/Male/Medium/Bodies/M_MED_CrownOrder/Meshes/M_MED_CrownOrder_Necklace_Physics
bodies {
 base_necklace [{shapeType=Capsule, length=1, radius=1, shapeName=base_necklace_capsule}]
 dyn_necklace_01 [{shapeType=Capsule, length=12, radius=2, shapeName=dyn_necklace_01_capsule, center=<1.5 0 0>}]
 dyn_necklace_02 [{shapeType=Capsule, length=10.5, radius=2, shapeName=dyn_necklace_02_capsule, center=<1.5 0 0>}]
 dyn_necklace_03 [{shapeType=Capsule, length=7.5, radius=1.7, shapeName=dyn_necklace_03_capsule, center=<1.5 0 0>}]
 dyn_necklace_04 [{shapeType=Capsule, length=5, radius=1.5, shapeName=dyn_necklace_04_capsule, center=<1.205 0 0>}]
 neck_01 [{shapeType=Capsule, length=12, radius=6, shapeName=neck_01_capsule, center=<-4.9856 -0.092 0>}]
 spine_05 [{shapeType=Capsule, length=8, radius=8.5, shapeName=spine_05_capsule, center=<-0.1967 0.7578 0>, rotation=<rot 90 180 -140.6338>}]
}
constraints [
 {bone1Name=dyn_necklace_02, bone2Name=dyn_necklace_01, swing1LimitDegrees=45, swing2LimitDegrees=20, twistLimitDegrees=5, twistMotion=Locked}
 {bone1Name=dyn_necklace_03, bone2Name=dyn_necklace_02, swing1LimitDegrees=45, swing2LimitDegrees=25, twistLimitDegrees=5, twistMotion=Locked}
 {bone1Name=dyn_necklace_04, bone2Name=dyn_necklace_03, swing1LimitDegrees=45, swing2LimitDegrees=25, twistLimitDegrees=5, twistMotion=Locked}
 {
  bone1Name=dyn_necklace_01
  bone2Name=base_necklace
  swing1LimitDegrees=45
  swing2LimitDegrees=45
  swing2Motion=Locked
  twistLimitDegrees=5
  twistMotion=Locked
 }
 {
  bone1Name=base_necklace
  bone2Name=neck_01
  swing1LimitDegrees=45
  swing1Motion=Free
  swing2LimitDegrees=45
  swing2Motion=Free
  twistLimitDegrees=45
  twistMotion=Free
 }
 {
  bone1Name=neck_01
  bone2Name=spine_05
  swing1LimitDegrees=45
  swing1Motion=Free
  swing2LimitDegrees=45
  swing2Motion=Free
  twistLimitDegrees=45
  twistMotion=Free
 }
]
