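# Post-process AnimBP for the CrownOrder player body, adding secondary motion to the costume over the pose copied from the attached parent mesh.
# A Control Rig takes movement, emote and front-end states, with hang gliding set by parachute events.
# Cached stages then run weighted bone constraints, shoulder/belt chain and necklace physics, collar, cuff and buckle AnimDynamics, and additive shoulder-chain offsets.
# The belt RigidBody node is unconnected.
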AnimBlueprint /BRCosmetics/Characters/Player/Male/Medium/Bodies/M_MED_CrownOrder/Meshes/M_MED_CrownOrder_AnimBP
target_skeleton=&/M_MED_CrownOrder_Skeleton
graphs [
 AnimGraph {nodes [
   RigidBody() {
    bClampLinearTranslationLimitToRefPose=true
    overridePhysicsAsset=&/M_MED_CrownOrder_Belt_Physics
    simulationSpace=BaseBoneSpace
    alphaBoolBlend {blendOption=Linear}
    baseBoneRef {boneName=thigh_l}
    componentAppliedLinearAccClamp=<3000 3000 3000>
    componentLinearAccScale=<0.6 0.6 0.6>
    componentLinearVelScale=<0.6 0.6 0.6>
    simSpaceSettings {
     maxAngularAcceleration=1000
     maxAngularVelocity=3
     maxLinearAcceleration=700
     maxLinearVelocity=500
     velocityScaleZ=0
     worldAlpha=1
     externalLinearDragV=<0.5 0.5 0.5>
    }
   }
   Output Pose(Result=Use cached pose 'outputCache'()) {layerGroup=None}
   n1 = Control Rig!(
    Source=Copy Pose From Mesh() {bUseAttachedParent=true}
    Alpha
    Crouching:bool=Get bIsCrouching()
    isHanggliding:bool=Get isHangGliding() -> (isHangGliding:bool)
    isSkyDivingFloat:bool=Get bIsSkydiveFloating()
    isEmoting:bool=Get bIsPlayingEmote()
    isSwimSprinting:bool=Get bIsSwimSprinting() -> (bIsSwimSprinting:bool)
    isFrontEnd:bool=Get bIsInFrontEnd()
   ) -> (Pose) {
    alphaBoolBlend {blendOption=Linear}
    controlRigAssetReference {blueprintRigClass=&/M_MED_CrownOrder_CtrlRig.M_MED_CrownOrder_CtrlRig_C}
   }
   Save cached pose 'inputCache'(Pose=@n1) {cacheName=inputCache}
   n2 = Anim Dynamics - Bone: dyn_collar_1_l!(ComponentPose=Local To Component(LocalPose=Use cached pose 'chainCache'()), ExternalForce="0,0,0") -> (Pose) {
    alpha=0.6
    angularDampingOverride=1
    angularSpringConstant=5000
    bAngularSpring=true
    bOverrideAngularDamping=true
    bUseGravityOverride=true
    gravityScale=0.35
    alphaBoolBlend {blendOption=Linear}
    boundBone {boneName=dyn_collar_1_l}
    componentAppliedLinearAccClamp=<10000 10000 10000>
    componentLinearAccScale=<1 1 1>
    componentLinearVelScale=<1 1 1>
    gravityOverride {z=-100}
    physicsBodyDefinitions [{
      boundBone {boneName=dyn_collar_1_l}
      boxExtents=<20 20 20>
      constraintSetup {angularLimitsMax=<0 45 15>, angularLimitsMin=<0 -45 -15>, angularTarget=<1 0 0>}
      localJointOffset=<2 0 0>
     }]
   }
   n3 = Anim Dynamics - Bone: dyn_collar_1_r!(ComponentPose=@n2, ExternalForce="0,0,0") -> (Pose) {
    alpha=0.6
    angularDampingOverride=1
    angularSpringConstant=5000
    bAngularSpring=true
    bOverrideAngularDamping=true
    bUseGravityOverride=true
    gravityScale=0.35
    alphaBoolBlend {blendOption=Linear}
    boundBone {boneName=dyn_collar_1_r}
    componentAppliedLinearAccClamp=<10000 10000 10000>
    componentLinearAccScale=<1 1 1>
    componentLinearVelScale=<1 1 1>
    gravityOverride {z=-100}
    physicsBodyDefinitions [{
      boundBone {boneName=dyn_collar_1_r}
      boxExtents=<20 20 20>
      constraintSetup {angularLimitsMax=<0 45 15>, angularLimitsMin=<0 -45 -15>, angularTarget=<1 0 0>}
      localJointOffset=<2 0 0>
     }]
   }
   n4 = Anim Dynamics - Bone: dyn_collar_2_l!(ComponentPose=@n3, ExternalForce="0,0,0") -> (Pose) {
    alpha=0.6
    angularDampingOverride=1
    angularSpringConstant=5000
    bAngularSpring=true
    bOverrideAngularDamping=true
    bUseGravityOverride=true
    gravityScale=0.35
    alphaBoolBlend {blendOption=Linear}
    boundBone {boneName=dyn_collar_2_l}
    componentAppliedLinearAccClamp=<10000 10000 10000>
    componentLinearAccScale=<1 1 1>
    componentLinearVelScale=<1 1 1>
    gravityOverride {z=-100}
    physicsBodyDefinitions [{
      boundBone {boneName=dyn_collar_2_l}
      boxExtents=<20 20 20>
      constraintSetup {angularLimitsMax=<0 45 15>, angularLimitsMin=<0 -45 -15>, angularTarget=<1 0 0>}
      localJointOffset=<2 0 0>
     }]
   }
   n5 = Anim Dynamics - Bone: dyn_collar_2_r!(ComponentPose=@n4, ExternalForce="0,0,0") -> (Pose) {
    alpha=0.6
    angularDampingOverride=1
    angularSpringConstant=5000
    bAngularSpring=true
    bOverrideAngularDamping=true
    bUseGravityOverride=true
    gravityScale=0.35
    alphaBoolBlend {blendOption=Linear}
    boundBone {boneName=dyn_collar_2_r}
    componentAppliedLinearAccClamp=<10000 10000 10000>
    componentLinearAccScale=<1 1 1>
    componentLinearVelScale=<1 1 1>
    gravityOverride {z=-100}
    physicsBodyDefinitions [{
      boundBone {boneName=dyn_collar_2_r}
      boxExtents=<20 20 20>
      constraintSetup {angularLimitsMax=<0 45 15>, angularLimitsMin=<0 -45 -15>, angularTarget=<1 0 0>}
      localJointOffset=<2 0 0>
     }]
   }
   n6 = Anim Dynamics - Bone: dyn_collar_3_l!(ComponentPose=@n5, ExternalForce="0,0,0") -> (Pose) {
    alpha=0.6
    angularDampingOverride=1
    angularSpringConstant=5000
    bAngularSpring=true
    bOverrideAngularDamping=true
    bUseGravityOverride=true
    gravityScale=0.35
    alphaBoolBlend {blendOption=Linear}
    boundBone {boneName=dyn_collar_3_l}
    componentAppliedLinearAccClamp=<10000 10000 10000>
    componentLinearAccScale=<1 1 1>
    componentLinearVelScale=<1 1 1>
    gravityOverride {z=-100}
    physicsBodyDefinitions [{
      boundBone {boneName=dyn_collar_3_l}
      boxExtents=<20 20 20>
      constraintSetup {angularLimitsMax=<0 45 15>, angularLimitsMin=<0 -45 -15>, angularTarget=<1 0 0>}
      localJointOffset=<2 0 0>
     }]
   }
   n7 = Anim Dynamics - Bone: dyn_collar_3_r!(ComponentPose=@n6, ExternalForce="0,0,0") -> (Pose) {
    alpha=0.6
    angularDampingOverride=1
    angularSpringConstant=5000
    bAngularSpring=true
    bOverrideAngularDamping=true
    bUseGravityOverride=true
    gravityScale=0.35
    alphaBoolBlend {blendOption=Linear}
    boundBone {boneName=dyn_collar_3_r}
    componentAppliedLinearAccClamp=<10000 10000 10000>
    componentLinearAccScale=<1 1 1>
    componentLinearVelScale=<1 1 1>
    gravityOverride {z=-100}
    physicsBodyDefinitions [{
      boundBone {boneName=dyn_collar_3_r}
      boxExtents=<20 20 20>
      constraintSetup {angularLimitsMax=<0 45 15>, angularLimitsMin=<0 -45 -15>, angularTarget=<1 0 0>}
      localJointOffset=<2 0 0>
     }]
   }
   n8 = Anim Dynamics - Bone: dyn_collar_1_c!(ComponentPose=@n7, ExternalForce="0,0,0") -> (Pose) {
    alpha=0.6
    angularDampingOverride=1
    angularSpringConstant=5000
    bAngularSpring=true
    bOverrideAngularDamping=true
    bUseGravityOverride=true
    gravityScale=0.35
    alphaBoolBlend {blendOption=Linear}
    boundBone {boneName=dyn_collar_1_c}
    componentAppliedLinearAccClamp=<10000 10000 10000>
    componentLinearAccScale=<1 1 1>
    componentLinearVelScale=<1 1 1>
    gravityOverride {z=-100}
    physicsBodyDefinitions [{
      boundBone {boneName=dyn_collar_1_c}
      boxExtents=<20 20 20>
      constraintSetup {angularLimitsMax=<0 45 15>, angularLimitsMin=<0 -45 -15>, angularTarget=<1 0 0>}
      localJointOffset=<2 0 0>
     }]
   }
   n9 = Anim Dynamics - Bone: dyn_cuff_l_01!(ComponentPose=@n8, ExternalForce="0,0,0") -> (Pose) {
    alpha=0.8
    angularDampingOverride=1
    angularSpringConstant=4500
    bAngularSpring=true
    bOverrideAngularDamping=true
    bUseGravityOverride=true
    gravityScale=0.35
    alphaBoolBlend {blendOption=Linear}
    boundBone {boneName=dyn_cuff_l_01}
    componentAppliedLinearAccClamp=<10000 10000 10000>
    componentLinearAccScale=<1 1 1>
    componentLinearVelScale=<1 1 1>
    gravityOverride {z=-100}
    physicsBodyDefinitions [{
      boundBone {boneName=dyn_cuff_l_01}
      boxExtents=<15 15 15>
      constraintSetup {angularLimitsMax=<0 15 45>, angularLimitsMin=<0 -15 -45>, angularTarget=<1 0 0>}
      localJointOffset=<1 0 0>
     }]
   }
   n10 = Anim Dynamics - Bone: dyn_cuff_l_01_end!(ComponentPose=@n9, ExternalForce="0,0,0") -> (Pose) {
    alpha=0.8
    angularDampingOverride=1
    angularSpringConstant=4500
    bAngularSpring=true
    bOverrideAngularDamping=true
    bUseGravityOverride=true
    gravityScale=0.35
    alphaBoolBlend {blendOption=Linear}
    boundBone {boneName=dyn_cuff_l_01_end}
    componentAppliedLinearAccClamp=<10000 10000 10000>
    componentLinearAccScale=<1 1 1>
    componentLinearVelScale=<1 1 1>
    gravityOverride {z=-100}
    physicsBodyDefinitions [{
      boundBone {boneName=dyn_cuff_l_01_end}
      boxExtents=<15 15 15>
      constraintSetup {angularLimitsMax=<0 15 45>, angularLimitsMin=<0 -15 -45>, angularTarget=<1 0 0>}
      localJointOffset=<1 0 0>
     }]
   }
   n11 = Anim Dynamics - Bone: dyn_cuff_l_02!(ComponentPose=@n10, ExternalForce="0,0,0") -> (Pose) {
    alpha=0.8
    angularDampingOverride=1
    angularSpringConstant=4500
    bAngularSpring=true
    bOverrideAngularDamping=true
    bUseGravityOverride=true
    gravityScale=0.35
    alphaBoolBlend {blendOption=Linear}
    boundBone {boneName=dyn_cuff_l_02}
    componentAppliedLinearAccClamp=<10000 10000 10000>
    componentLinearAccScale=<1 1 1>
    componentLinearVelScale=<1 1 1>
    gravityOverride {z=-100}
    physicsBodyDefinitions [{
      boundBone {boneName=dyn_cuff_l_02}
      boxExtents=<15 15 15>
      constraintSetup {angularLimitsMax=<0 45 15>, angularLimitsMin=<0 -45 -15>, angularTarget=<1 0 0>}
      localJointOffset=<1 0 0>
     }]
   }
   n12 = Anim Dynamics - Bone: dyn_cuff_l_02_end!(ComponentPose=@n11, ExternalForce="0,0,0") -> (Pose) {
    alpha=0.8
    angularDampingOverride=1
    angularSpringConstant=4500
    bAngularSpring=true
    bOverrideAngularDamping=true
    bUseGravityOverride=true
    gravityScale=0.35
    alphaBoolBlend {blendOption=Linear}
    boundBone {boneName=dyn_cuff_l_02_end}
    componentAppliedLinearAccClamp=<10000 10000 10000>
    componentLinearAccScale=<1 1 1>
    componentLinearVelScale=<1 1 1>
    gravityOverride {z=-100}
    physicsBodyDefinitions [{
      boundBone {boneName=dyn_cuff_l_02_end}
      boxExtents=<15 15 15>
      constraintSetup {angularLimitsMax=<0 45 15>, angularLimitsMin=<0 -45 -15>, angularTarget=<1 0 0>}
      localJointOffset=<1 0 0>
     }]
   }
   n13 = Anim Dynamics - Bone: dyn_cuff_l_03!(ComponentPose=@n12, ExternalForce="0,0,0") -> (Pose) {
    alpha=0.8
    angularDampingOverride=1
    angularSpringConstant=4500
    bAngularSpring=true
    bOverrideAngularDamping=true
    bUseGravityOverride=true
    gravityScale=0.35
    alphaBoolBlend {blendOption=Linear}
    boundBone {boneName=dyn_cuff_l_03}
    componentAppliedLinearAccClamp=<10000 10000 10000>
    componentLinearAccScale=<1 1 1>
    componentLinearVelScale=<1 1 1>
    gravityOverride {z=-100}
    physicsBodyDefinitions [{
      boundBone {boneName=dyn_cuff_l_03}
      boxExtents=<15 15 15>
      constraintSetup {angularLimitsMax=<0 15 45>, angularLimitsMin=<0 -15 -45>, angularTarget=<1 0 0>}
      localJointOffset=<1 0 0>
     }]
   }
   n14 = Anim Dynamics - Bone: dyn_cuff_l_03_end!(ComponentPose=@n13, ExternalForce="0,0,0") -> (Pose) {
    alpha=0.8
    angularDampingOverride=1
    angularSpringConstant=4500
    bAngularSpring=true
    bOverrideAngularDamping=true
    bUseGravityOverride=true
    gravityScale=0.35
    alphaBoolBlend {blendOption=Linear}
    boundBone {boneName=dyn_cuff_l_03_end}
    componentAppliedLinearAccClamp=<10000 10000 10000>
    componentLinearAccScale=<1 1 1>
    componentLinearVelScale=<1 1 1>
    gravityOverride {z=-100}
    physicsBodyDefinitions [{
      boundBone {boneName=dyn_cuff_l_03_end}
      boxExtents=<15 15 15>
      constraintSetup {angularLimitsMax=<0 15 45>, angularLimitsMin=<0 -15 -45>, angularTarget=<1 0 0>}
      localJointOffset=<1 0 0>
     }]
   }
   n15 = Anim Dynamics - Bone: dyn_cuff_l_04!(ComponentPose=@n14, ExternalForce="0,0,0") -> (Pose) {
    alpha=0.8
    angularDampingOverride=1
    angularSpringConstant=4500
    bAngularSpring=true
    bOverrideAngularDamping=true
    bUseGravityOverride=true
    gravityScale=0.35
    alphaBoolBlend {blendOption=Linear}
    boundBone {boneName=dyn_cuff_l_04}
    componentAppliedLinearAccClamp=<10000 10000 10000>
    componentLinearAccScale=<1 1 1>
    componentLinearVelScale=<1 1 1>
    gravityOverride {z=-100}
    physicsBodyDefinitions [{
      boundBone {boneName=dyn_cuff_l_04}
      boxExtents=<15 15 15>
      constraintSetup {angularLimitsMax=<0 15 45>, angularLimitsMin=<0 -15 -45>, angularTarget=<1 0 0>}
      localJointOffset=<1 0 0>
     }]
   }
   n16 = Anim Dynamics - Bone: dyn_cuff_l_04_end!(ComponentPose=@n15, ExternalForce="0,0,0") -> (Pose) {
    alpha=0.8
    angularDampingOverride=1
    angularSpringConstant=4500
    bAngularSpring=true
    bOverrideAngularDamping=true
    bUseGravityOverride=true
    gravityScale=0.35
    alphaBoolBlend {blendOption=Linear}
    boundBone {boneName=dyn_cuff_l_04_end}
    componentAppliedLinearAccClamp=<10000 10000 10000>
    componentLinearAccScale=<1 1 1>
    componentLinearVelScale=<1 1 1>
    gravityOverride {z=-100}
    physicsBodyDefinitions [{
      boundBone {boneName=dyn_cuff_l_04_end}
      boxExtents=<15 15 15>
      constraintSetup {angularLimitsMax=<0 15 45>, angularLimitsMin=<0 -15 -45>, angularTarget=<1 0 0>}
      localJointOffset=<1 0 0>
     }]
   }
   n17 = Anim Dynamics - Bone: dyn_cuff_r_01!(ComponentPose=@n16, ExternalForce="0,0,0") -> (Pose) {
    alpha=0.8
    angularDampingOverride=1
    angularSpringConstant=4500
    bAngularSpring=true
    bOverrideAngularDamping=true
    bUseGravityOverride=true
    gravityScale=0.35
    alphaBoolBlend {blendOption=Linear}
    boundBone {boneName=dyn_cuff_r_01}
    componentAppliedLinearAccClamp=<10000 10000 10000>
    componentLinearAccScale=<1 1 1>
    componentLinearVelScale=<1 1 1>
    gravityOverride {z=-100}
    physicsBodyDefinitions [{
      boundBone {boneName=dyn_cuff_r_01}
      boxExtents=<15 15 15>
      constraintSetup {angularLimitsMax=<0 15 45>, angularLimitsMin=<0 -15 -45>, angularTarget=<1 0 0>}
      localJointOffset=<1 0 0>
     }]
   }
   n18 = Anim Dynamics - Bone: dyn_cuff_r_01_end!(ComponentPose=@n17, ExternalForce="0,0,0") -> (Pose) {
    alpha=0.8
    angularDampingOverride=1
    angularSpringConstant=4500
    bAngularSpring=true
    bOverrideAngularDamping=true
    bUseGravityOverride=true
    gravityScale=0.35
    alphaBoolBlend {blendOption=Linear}
    boundBone {boneName=dyn_cuff_r_01_end}
    componentAppliedLinearAccClamp=<10000 10000 10000>
    componentLinearAccScale=<1 1 1>
    componentLinearVelScale=<1 1 1>
    gravityOverride {z=-100}
    physicsBodyDefinitions [{
      boundBone {boneName=dyn_cuff_r_01_end}
      boxExtents=<15 15 15>
      constraintSetup {angularLimitsMax=<0 15 45>, angularLimitsMin=<0 -15 -45>, angularTarget=<1 0 0>}
      localJointOffset=<1 0 0>
     }]
   }
   n19 = Anim Dynamics - Bone: dyn_cuff_r_02!(ComponentPose=@n18, ExternalForce="0,0,0") -> (Pose) {
    alpha=0.8
    angularDampingOverride=1
    angularSpringConstant=4500
    bAngularSpring=true
    bOverrideAngularDamping=true
    bUseGravityOverride=true
    gravityScale=0.35
    alphaBoolBlend {blendOption=Linear}
    boundBone {boneName=dyn_cuff_r_02}
    componentAppliedLinearAccClamp=<10000 10000 10000>
    componentLinearAccScale=<1 1 1>
    componentLinearVelScale=<1 1 1>
    gravityOverride {z=-100}
    physicsBodyDefinitions [{
      boundBone {boneName=dyn_cuff_r_02}
      boxExtents=<15 15 15>
      constraintSetup {angularLimitsMax=<0 45 15>, angularLimitsMin=<0 -45 -15>, angularTarget=<1 0 0>}
      localJointOffset=<1 0 0>
     }]
   }
   n20 = Anim Dynamics - Bone: dyn_cuff_r_02_end!(ComponentPose=@n19, ExternalForce="0,0,0") -> (Pose) {
    alpha=0.8
    angularDampingOverride=1
    angularSpringConstant=4500
    bAngularSpring=true
    bOverrideAngularDamping=true
    bUseGravityOverride=true
    gravityScale=0.35
    alphaBoolBlend {blendOption=Linear}
    boundBone {boneName=dyn_cuff_r_02_end}
    componentAppliedLinearAccClamp=<10000 10000 10000>
    componentLinearAccScale=<1 1 1>
    componentLinearVelScale=<1 1 1>
    gravityOverride {z=-100}
    physicsBodyDefinitions [{
      boundBone {boneName=dyn_cuff_r_02_end}
      boxExtents=<15 15 15>
      constraintSetup {angularLimitsMax=<0 45 15>, angularLimitsMin=<0 -45 -15>, angularTarget=<1 0 0>}
      localJointOffset=<1 0 0>
     }]
   }
   n21 = Anim Dynamics - Bone: dyn_cuff_r_03!(ComponentPose=@n20, ExternalForce="0,0,0") -> (Pose) {
    alpha=0.8
    angularDampingOverride=1
    angularSpringConstant=4500
    bAngularSpring=true
    bOverrideAngularDamping=true
    bUseGravityOverride=true
    gravityScale=0.35
    alphaBoolBlend {blendOption=Linear}
    boundBone {boneName=dyn_cuff_r_03}
    componentAppliedLinearAccClamp=<10000 10000 10000>
    componentLinearAccScale=<1 1 1>
    componentLinearVelScale=<1 1 1>
    gravityOverride {z=-100}
    physicsBodyDefinitions [{
      boundBone {boneName=dyn_cuff_r_03}
      boxExtents=<15 15 15>
      constraintSetup {angularLimitsMax=<0 15 45>, angularLimitsMin=<0 -15 -45>, angularTarget=<1 0 0>}
      localJointOffset=<1 0 0>
     }]
   }
   n22 = Anim Dynamics - Bone: dyn_cuff_r_03_end!(ComponentPose=@n21, ExternalForce="0,0,0") -> (Pose) {
    alpha=0.8
    angularDampingOverride=1
    angularSpringConstant=4500
    bAngularSpring=true
    bOverrideAngularDamping=true
    bUseGravityOverride=true
    gravityScale=0.35
    alphaBoolBlend {blendOption=Linear}
    boundBone {boneName=dyn_cuff_r_03_end}
    componentAppliedLinearAccClamp=<10000 10000 10000>
    componentLinearAccScale=<1 1 1>
    componentLinearVelScale=<1 1 1>
    gravityOverride {z=-100}
    physicsBodyDefinitions [{
      boundBone {boneName=dyn_cuff_r_03_end}
      boxExtents=<15 15 15>
      constraintSetup {angularLimitsMax=<0 15 45>, angularLimitsMin=<0 -15 -45>, angularTarget=<1 0 0>}
      localJointOffset=<1 0 0>
     }]
   }
   n23 = Anim Dynamics - Bone: dyn_cuff_r_04!(ComponentPose=@n22, ExternalForce="0,0,0") -> (Pose) {
    alpha=0.8
    angularDampingOverride=1
    angularSpringConstant=4500
    bAngularSpring=true
    bOverrideAngularDamping=true
    bUseGravityOverride=true
    gravityScale=0.35
    alphaBoolBlend {blendOption=Linear}
    boundBone {boneName=dyn_cuff_r_04}
    componentAppliedLinearAccClamp=<10000 10000 10000>
    componentLinearAccScale=<1 1 1>
    componentLinearVelScale=<1 1 1>
    gravityOverride {z=-100}
    physicsBodyDefinitions [{
      boundBone {boneName=dyn_cuff_r_04}
      boxExtents=<15 15 15>
      constraintSetup {angularLimitsMax=<0 15 45>, angularLimitsMin=<0 -15 -45>, angularTarget=<1 0 0>}
      localJointOffset=<1 0 0>
     }]
   }
   n24 = Anim Dynamics - Bone: dyn_cuff_r_04_end!(ComponentPose=@n23, ExternalForce="0,0,0") -> (Pose) {
    alpha=0.8
    angularDampingOverride=1
    angularSpringConstant=4500
    bAngularSpring=true
    bOverrideAngularDamping=true
    bUseGravityOverride=true
    gravityScale=0.35
    alphaBoolBlend {blendOption=Linear}
    boundBone {boneName=dyn_cuff_r_04_end}
    componentAppliedLinearAccClamp=<10000 10000 10000>
    componentLinearAccScale=<1 1 1>
    componentLinearVelScale=<1 1 1>
    gravityOverride {z=-100}
    physicsBodyDefinitions [{
      boundBone {boneName=dyn_cuff_r_04_end}
      boxExtents=<15 15 15>
      constraintSetup {angularLimitsMax=<0 15 45>, angularLimitsMin=<0 -15 -45>, angularTarget=<1 0 0>}
      localJointOffset=<1 0 0>
     }]
   }
   n25 = Anim Dynamics - Bone: dyn_buckle_l_01_up!(ComponentPose=@n24, ExternalForce="0,0,0") -> (Pose) {
    alpha=0.8
    angularDampingOverride=1
    angularSpringConstant=5000
    bAngularSpring=true
    bOverrideAngularDamping=true
    bUseGravityOverride=true
    gravityScale=0.35
    alphaBoolBlend {blendOption=Linear}
    boundBone {boneName=dyn_buckle_l_01_up}
    componentAppliedLinearAccClamp=<10000 10000 10000>
    componentLinearAccScale=<1 1 1>
    componentLinearVelScale=<1 1 1>
    gravityOverride {z=-980}
    physicsBodyDefinitions [{
      boundBone {boneName=dyn_buckle_l_01_up}
      boxExtents=<15 15 15>
      constraintSetup {angularLimitsMax=<0 45 20>, angularLimitsMin=<0 -30 -20>, angularTarget=<1 0 0>}
      localJointOffset=<2 0 0>
     }]
    relativeSpaceBone {boneName=spine_02}
   }
   n26 = Anim Dynamics - Bone: dyn_buckle_l_02_up!(ComponentPose=@n25, ExternalForce="0,0,0") -> (Pose) {
    alpha=0.8
    angularDampingOverride=1
    angularSpringConstant=5000
    bAngularSpring=true
    bOverrideAngularDamping=true
    bUseGravityOverride=true
    gravityScale=0.35
    alphaBoolBlend {blendOption=Linear}
    boundBone {boneName=dyn_buckle_l_02_up}
    componentAppliedLinearAccClamp=<10000 10000 10000>
    componentLinearAccScale=<1 1 1>
    componentLinearVelScale=<1 1 1>
    gravityOverride {z=-980}
    physicsBodyDefinitions [{
      boundBone {boneName=dyn_buckle_l_02_up}
      boxExtents=<15 15 15>
      constraintSetup {angularLimitsMax=<0 45 20>, angularLimitsMin=<0 -30 -20>, angularTarget=<1 0 0>}
      localJointOffset=<2 0 0>
     }]
    relativeSpaceBone {boneName=spine_02}
   }
   n27 = Anim Dynamics - Bone: dyn_buckle_l_01_dw!(ComponentPose=@n26, ExternalForce="0,0,0") -> (Pose) {
    alpha=0.8
    angularDampingOverride=1
    angularSpringConstant=5000
    bAngularSpring=true
    bOverrideAngularDamping=true
    bUseGravityOverride=true
    gravityScale=0.35
    alphaBoolBlend {blendOption=Linear}
    boundBone {boneName=dyn_buckle_l_01_dw}
    componentAppliedLinearAccClamp=<10000 10000 10000>
    componentLinearAccScale=<1 1 1>
    componentLinearVelScale=<1 1 1>
    gravityOverride {z=-980}
    physicsBodyDefinitions [{
      boundBone {boneName=dyn_buckle_l_01_dw}
      boxExtents=<15 15 15>
      constraintSetup {angularLimitsMax=<0 45 20>, angularLimitsMin=<0 -30 -20>, angularTarget=<1 0 0>}
      localJointOffset=<2 0 0>
     }]
    relativeSpaceBone {boneName=spine_02}
   }
   n28 = Anim Dynamics - Bone: dyn_buckle_l_02_dw!(ComponentPose=@n27, ExternalForce="0,0,0") -> (Pose) {
    alpha=0.8
    angularDampingOverride=1
    angularSpringConstant=5000
    bAngularSpring=true
    bOverrideAngularDamping=true
    bUseGravityOverride=true
    gravityScale=0.35
    alphaBoolBlend {blendOption=Linear}
    boundBone {boneName=dyn_buckle_l_02_dw}
    componentAppliedLinearAccClamp=<10000 10000 10000>
    componentLinearAccScale=<1 1 1>
    componentLinearVelScale=<1 1 1>
    gravityOverride {z=-980}
    physicsBodyDefinitions [{
      boundBone {boneName=dyn_buckle_l_02_dw}
      boxExtents=<15 15 15>
      constraintSetup {angularLimitsMax=<0 45 20>, angularLimitsMin=<0 -30 -20>, angularTarget=<1 0 0>}
      localJointOffset=<2 0 0>
     }]
    relativeSpaceBone {boneName=spine_02}
   }
   n29 = Anim Dynamics - Bone: dyn_buckle_r_01_up!(ComponentPose=@n28, ExternalForce="0,0,0") -> (Pose) {
    alpha=0.8
    angularDampingOverride=1
    angularSpringConstant=5000
    bAngularSpring=true
    bOverrideAngularDamping=true
    bUseGravityOverride=true
    gravityScale=0.35
    alphaBoolBlend {blendOption=Linear}
    boundBone {boneName=dyn_buckle_r_01_up}
    componentAppliedLinearAccClamp=<10000 10000 10000>
    componentLinearAccScale=<1 1 1>
    componentLinearVelScale=<1 1 1>
    gravityOverride {z=-980}
    physicsBodyDefinitions [{
      boundBone {boneName=dyn_buckle_r_01_up}
      boxExtents=<15 15 15>
      constraintSetup {angularLimitsMax=<0 20 45>, angularLimitsMin=<0 -20 -30>, angularTarget=<1 0 0>}
      localJointOffset=<2 0 0>
     }]
    relativeSpaceBone {boneName=spine_02}
   }
   n30 = Anim Dynamics - Bone: dyn_buckle_r_02_up!(ComponentPose=@n29, ExternalForce="0,0,0") -> (Pose) {
    alpha=0.8
    angularDampingOverride=1
    angularSpringConstant=5000
    bAngularSpring=true
    bOverrideAngularDamping=true
    bUseGravityOverride=true
    gravityScale=0.35
    alphaBoolBlend {blendOption=Linear}
    boundBone {boneName=dyn_buckle_r_02_up}
    componentAppliedLinearAccClamp=<10000 10000 10000>
    componentLinearAccScale=<1 1 1>
    componentLinearVelScale=<1 1 1>
    gravityOverride {z=-980}
    physicsBodyDefinitions [{
      boundBone {boneName=dyn_buckle_r_02_up}
      boxExtents=<15 15 15>
      constraintSetup {angularLimitsMax=<0 20 45>, angularLimitsMin=<0 -20 -30>, angularTarget=<1 0 0>}
      localJointOffset=<2 0 0>
     }]
    relativeSpaceBone {boneName=spine_02}
   }
   n31 = Anim Dynamics - Bone: dyn_buckle_r_01_dw!(ComponentPose=@n30, ExternalForce="0,0,0") -> (Pose) {
    alpha=0.8
    angularDampingOverride=1
    angularSpringConstant=5000
    bAngularSpring=true
    bOverrideAngularDamping=true
    bUseGravityOverride=true
    gravityScale=0.35
    alphaBoolBlend {blendOption=Linear}
    boundBone {boneName=dyn_buckle_r_01_dw}
    componentAppliedLinearAccClamp=<10000 10000 10000>
    componentLinearAccScale=<1 1 1>
    componentLinearVelScale=<1 1 1>
    gravityOverride {z=-980}
    physicsBodyDefinitions [{
      boundBone {boneName=dyn_buckle_r_01_dw}
      boxExtents=<15 15 15>
      constraintSetup {angularLimitsMax=<0 20 45>, angularLimitsMin=<0 -20 -30>, angularTarget=<1 0 0>}
      localJointOffset=<2 0 0>
     }]
    relativeSpaceBone {boneName=spine_02}
   }
   n32 = Anim Dynamics - Bone: dyn_buckle_r_02_dw!(ComponentPose=@n31, ExternalForce="0,0,0") -> (Pose) {
    alpha=0.8
    angularDampingOverride=1
    angularSpringConstant=5000
    bAngularSpring=true
    bOverrideAngularDamping=true
    bUseGravityOverride=true
    gravityScale=0.35
    alphaBoolBlend {blendOption=Linear}
    boundBone {boneName=dyn_buckle_r_02_dw}
    componentAppliedLinearAccClamp=<10000 10000 10000>
    componentLinearAccScale=<1 1 1>
    componentLinearVelScale=<1 1 1>
    gravityOverride {z=-980}
    physicsBodyDefinitions [{
      boundBone {boneName=dyn_buckle_r_02_dw}
      boxExtents=<15 15 15>
      constraintSetup {angularLimitsMax=<0 20 45>, angularLimitsMin=<0 -20 -30>, angularTarget=<1 0 0>}
      localJointOffset=<2 0 0>
     }]
    relativeSpaceBone {boneName=spine_02}
   }
   n33 = "Transform (Modify) Bone - Bone: base_shoulderChainB_dw"!(ComponentPose=@n32, bAlphaBoolEnabled:bool=True, Rotation="14,-10,20") -> (Pose) {
    alphaInputType=Bool
    rotationMode=BMM_Additive
    alphaBoolBlend {blendOption=Linear}
    boneToModify {boneName=base_shoulderChainB_dw}
    rotation=<rot 14 -10 20>
   }
   n34 = "Transform (Modify) Bone - Bone: base_shoulderChainA_dw"!(ComponentPose=@n33, bAlphaBoolEnabled:bool=True, Rotation="5,0,-20") -> (Pose) {
    alphaInputType=Bool
    rotationMode=BMM_Additive
    alphaBoolBlend {blendOption=Linear}
    boneToModify {boneName=base_shoulderChainA_dw}
    rotation {pitch=5, roll=-20}
   }
   n35 = "Transform (Modify) Bone - Bone: base_shoulderChainA_up"!(ComponentPose=@n34, bAlphaBoolEnabled:bool=True, Rotation="2,0,0") -> (Pose) {
    alphaInputType=Bool
    rotationMode=BMM_Additive
    alphaBoolBlend {blendOption=Linear}
    boneToModify {boneName=base_shoulderChainA_up}
    rotation {pitch=2}
   }
   n36 = "Transform (Modify) Bone - Bone: base_shoulderChainB_up"!(ComponentPose=@n35, bAlphaBoolEnabled:bool=True, Rotation="6,0,10") -> (Pose) {
    alphaInputType=Bool
    rotationMode=BMM_Additive
    alphaBoolBlend {blendOption=Linear}
    boneToModify {boneName=base_shoulderChainB_up}
    rotation {pitch=6, roll=10}
   }
   Save cached pose 'outputCache'(Pose=Component To Local(ComponentPose=@n36)) {cacheName=outputCache}
   n37 = "Constraint - dyn_beltChain_con (PPPPP)"!(
    ComponentPose=Local To Component(LocalPose=Use cached pose 'inputCache'())
    Alpha=1
    ConstraintWeights_0=0.4959
    ConstraintWeights_1:float=0.2646
    ConstraintWeights_2:float=0.2347
    ConstraintWeights_3:float=0.0032
    ConstraintWeights_4:float=0.0015
   ) -> (Pose) {
    alphaBoolBlend {blendOption=Linear}
    boneToModify {boneName=dyn_beltChain_con}
    constraintSetup [
     {targetBone {boneName=pelvis}}
     {targetBone {boneName=deform_groin_r}}
     {targetBone {boneName=thigh_twist_01_r}}
     {targetBone {boneName=thigh_twist_01_l}}
     {targetBone {boneName=spine_02}}
    ]
    constraintWeights [1,1,1,1,1]
   }
   n38 = "Constraint - base_buckle_l_up (PPPPPPPP)"!(
    ComponentPose=@n37
    Alpha=1
    ConstraintWeights_0=0.3083
    ConstraintWeights_1:float=0.1991
    ConstraintWeights_2:float=0.1627
    ConstraintWeights_3:float=0.1562
    ConstraintWeights_4:float=0.1059
    ConstraintWeights_5:float=0.0543
    ConstraintWeights_6:float=0.0108
    ConstraintWeights_7:float=0.0027
   ) -> (Pose) {
    alphaBoolBlend {blendOption=Linear}
    boneToModify {boneName=base_buckle_l_up}
    constraintSetup [
     {targetBone {boneName=spine_03}}
     {targetBone {boneName=spine_04}}
     {targetBone {boneName=deform_pec_l}}
     {targetBone {boneName=spine_02}}
     {targetBone {boneName=clavicle_l}}
     {targetBone {boneName=spine_01}}
     {targetBone {boneName=spine_05}}
     {targetBone {boneName=upperarm_twist_01_l}}
    ]
    constraintWeights [1,1,1,1,1,1,1,1]
   }
   n39 = "Constraint - base_buckle_l_dw (PPPPPPPP)"!(
    ComponentPose=@n38
    Alpha=1
    ConstraintWeights_0=0.34
    ConstraintWeights_1:float=0.2165
    ConstraintWeights_2:float=0.1488
    ConstraintWeights_3:float=0.0952
    ConstraintWeights_4:float=0.094
    ConstraintWeights_5:float=0.072
    ConstraintWeights_6:float=0.0295
    ConstraintWeights_7:float=0.0042
   ) -> (Pose) {
    alphaBoolBlend {blendOption=Linear}
    boneToModify {boneName=base_buckle_l_dw}
    constraintSetup [
     {targetBone {boneName=spine_03}}
     {targetBone {boneName=spine_02}}
     {targetBone {boneName=spine_04}}
     {targetBone {boneName=clavicle_l}}
     {targetBone {boneName=spine_01}}
     {targetBone {boneName=deform_pec_l}}
     {targetBone {boneName=pelvis}}
     {targetBone {boneName=upperarm_twist_01_l}}
    ]
    constraintWeights [1,1,1,1,1,1,1,1]
   }
   n40 = "Constraint - base_buckle_r_up (PPPPPPPP)"!(
    ComponentPose=@n39
    Alpha=1
    ConstraintWeights_0=0.2928
    ConstraintWeights_1:float=0.2093
    ConstraintWeights_2:float=0.178
    ConstraintWeights_3:float=0.1452
    ConstraintWeights_4:float=0.104
    ConstraintWeights_5:float=0.0472
    ConstraintWeights_6:float=0.0214
    ConstraintWeights_7:float=0.0022
   ) -> (Pose) {
    alphaBoolBlend {blendOption=Linear}
    boneToModify {boneName=base_buckle_r_up}
    constraintSetup [
     {targetBone {boneName=spine_03}}
     {targetBone {boneName=spine_04}}
     {targetBone {boneName=deform_pec_r}}
     {targetBone {boneName=spine_02}}
     {targetBone {boneName=clavicle_r}}
     {targetBone {boneName=spine_01}}
     {targetBone {boneName=spine_05}}
     {targetBone {boneName=upperarm_twist_01_r}}
    ]
    constraintWeights [1,1,1,1,1,1,1,1]
   }
   n41 = "Constraint - base_buckle_r_dw (PPPPPPP)"!(
    ComponentPose=@n40
    Alpha=1
    ConstraintWeights_0=0.3394
    ConstraintWeights_1:float=0.2062
    ConstraintWeights_2:float=0.1573
    ConstraintWeights_3:float=0.0988
    ConstraintWeights_4:float=0.0873
    ConstraintWeights_5:float=0.0867
    ConstraintWeights_6:float=0.0243
   ) -> (Pose) {
    alphaBoolBlend {blendOption=Linear}
    boneToModify {boneName=base_buckle_r_dw}
    constraintSetup [
     {targetBone {boneName=spine_03}}
     {targetBone {boneName=spine_02}}
     {targetBone {boneName=spine_04}}
     {targetBone {boneName=clavicle_r}}
     {targetBone {boneName=spine_01}}
     {targetBone {boneName=deform_pec_r}}
     {targetBone {boneName=pelvis}}
    ]
    constraintWeights [1,1,1,1,1,1,1]
   }
   Save cached pose 'parentConstraintCache'(Pose=Component To Local(ComponentPose=@n41)) {cacheName=parentConstraintCache}
   n42 = Anim Dynamics - Bone: dyn_shoulderChain_A01_up(ComponentPose=Local To Component(LocalPose=Use cached pose 'parentConstraintCache'()), Alpha=1, ExternalForce="0,0,0") {
    angularDampingOverride=1
    angularSpringConstant=3500
    bAngularSpring=true
    bOverrideAngularDamping=true
    bUseGravityOverride=true
    gravityScale=0.35
    simulationSpace=BoneRelative
    alphaBoolBlend {blendOption=Linear}
    boundBone {boneName=dyn_shoulderChain_A01_up}
    componentAppliedLinearAccClamp=<10000 10000 10000>
    componentLinearAccScale=<1 1 1>
    componentLinearVelScale=<1 1 1>
    gravityOverride {z=-980}
    physicsBodyDefinitions [{
      boundBone {boneName=dyn_shoulderChain_A01_up}
      boxExtents=<20 25 25>
      constraintSetup {angularLimitsMax=<1 1 10>, angularLimitsMin=<-2 -30 -1>, angularTarget=<1 0 0>}
      localJointOffset=<2 0 0>
     }]
    relativeSpaceBone {boneName=clavicle_r}
   }
   n43 = Anim Dynamics - Bone: dyn_shoulderChain_B01_up(ComponentPose=@n42, Alpha=1, ExternalForce="0,0,0") {
    angularDampingOverride=1
    angularSpringConstant=3500
    bAngularSpring=true
    bOverrideAngularDamping=true
    bUseGravityOverride=true
    gravityScale=0.35
    simulationSpace=BoneRelative
    alphaBoolBlend {blendOption=Linear}
    boundBone {boneName=dyn_shoulderChain_B01_up}
    componentAppliedLinearAccClamp=<10000 10000 10000>
    componentLinearAccScale=<1 1 1>
    componentLinearVelScale=<1 1 1>
    gravityOverride {z=-980}
    physicsBodyDefinitions [{
      boundBone {boneName=dyn_shoulderChain_B01_up}
      boxExtents=<20 25 25>
      constraintSetup {angularLimitsMax=<1 1 1>, angularLimitsMin=<-2 -30 -10>, angularTarget=<1 0 0>}
      localJointOffset=<2 0 0>
     }]
    relativeSpaceBone {boneName=clavicle_r}
   }
   n44 = Anim Dynamics - Bone: dyn_shoulderChain_A01_dw(ComponentPose=@n43, Alpha=1, ExternalForce="0,0,0") {
    angularDampingOverride=1
    angularSpringConstant=3500
    bAngularSpring=true
    bOverrideAngularDamping=true
    bUseGravityOverride=true
    gravityScale=0.35
    simulationSpace=BoneRelative
    alphaBoolBlend {blendOption=Linear}
    boundBone {boneName=dyn_shoulderChain_A01_dw}
    componentAppliedLinearAccClamp=<10000 10000 10000>
    componentLinearAccScale=<1 1 1>
    componentLinearVelScale=<1 1 1>
    gravityOverride {z=-980}
    physicsBodyDefinitions [{
      boundBone {boneName=dyn_shoulderChain_A01_dw}
      boxExtents=<18 22 22>
      constraintSetup {angularLimitsMax=<1 1 10>, angularLimitsMin=<-2 -30 -1>, angularTarget=<1 0 0>}
      localJointOffset=<2 0 0>
     }]
    relativeSpaceBone {boneName=clavicle_r}
   }
   n45 = Anim Dynamics - Bone: dyn_shoulderChain_B01_dw(ComponentPose=@n44, Alpha=1, ExternalForce="0,0,0") {
    angularDampingOverride=1
    angularSpringConstant=3500
    bAngularSpring=true
    bOverrideAngularDamping=true
    bUseGravityOverride=true
    gravityScale=0.35
    simulationSpace=BoneRelative
    alphaBoolBlend {blendOption=Linear}
    boundBone {boneName=dyn_shoulderChain_B01_dw}
    componentAppliedLinearAccClamp=<10000 10000 10000>
    componentLinearAccScale=<1 1 1>
    componentLinearVelScale=<1 1 1>
    gravityOverride {z=-980}
    physicsBodyDefinitions [{
      boundBone {boneName=dyn_shoulderChain_B01_dw}
      boxExtents=<18 22 22>
      constraintSetup {angularLimitsMax=<1 1 1>, angularLimitsMin=<-2 -30 -10>, angularTarget=<1 0 0>}
      localJointOffset=<2 0 0>
     }]
    relativeSpaceBone {boneName=clavicle_r}
   }
   n46 = RigidBody(ComponentPose=@n45, Alpha=0.8) {
    bClampLinearTranslationLimitToRefPose=true
    overridePhysicsAsset=&/M_MED_CrownOrder_Necklace_Physics
    simulationSpace=BaseBoneSpace
    alphaBoolBlend {blendOption=Linear}
    baseBoneRef {boneName=spine_05}
    componentAppliedLinearAccClamp=<3000 3000 3000>
    componentLinearAccScale=<0.6 0.6 0.6>
    componentLinearVelScale=<0.6 0.6 0.6>
    simSpaceSettings {
     maxAngularAcceleration=1000
     maxAngularVelocity=3
     maxLinearAcceleration=700
     maxLinearVelocity=500
     velocityScaleZ=0
     worldAlpha=1
     externalLinearDragV=<0.5 0.5 0.5>
    }
   }
   n47 = Anim Dynamics - Bone: dyn_beltChain_01(ComponentPose=@n46, Alpha=1, ExternalForce="0,0,0") {
    angularDampingOverride=1
    angularSpringConstant=3500
    bAngularSpring=true
    bOverrideAngularDamping=true
    bUseGravityOverride=true
    gravityScale=0.35
    simulationSpace=BoneRelative
    alphaBoolBlend {blendOption=Linear}
    boundBone {boneName=dyn_beltChain_01}
    componentAppliedLinearAccClamp=<10000 10000 10000>
    componentLinearAccScale=<1 1 1>
    componentLinearVelScale=<1 1 1>
    gravityOverride {z=-980}
    physicsBodyDefinitions [{
      boundBone {boneName=dyn_beltChain_01}
      boxExtents=<15 15 15>
      constraintSetup {angularLimitsMax=<0 0 45>, angularLimitsMin=<0 0 -5>, angularTarget=<1 0 0>}
      localJointOffset=<2 0 0>
     }]
    relativeSpaceBone {boneName=thigh_r}
   }
   n48 = Anim Dynamics - Bone: dyn_beltChain_02(ComponentPose=@n47, Alpha=1, ExternalForce="0,0,0") {
    angularDampingOverride=1
    angularSpringConstant=3500
    bAngularSpring=true
    bOverrideAngularDamping=true
    bUseGravityOverride=true
    gravityScale=0.35
    simulationSpace=BoneRelative
    alphaBoolBlend {blendOption=Linear}
    boundBone {boneName=dyn_beltChain_02}
    componentAppliedLinearAccClamp=<10000 10000 10000>
    componentLinearAccScale=<1 1 1>
    componentLinearVelScale=<1 1 1>
    gravityOverride {z=-980}
    physicsBodyDefinitions [{
      boundBone {boneName=dyn_beltChain_02}
      boxExtents=<15 15 15>
      constraintSetup {angularLimitsMax=<0 15 45>, angularLimitsMin=<0 -15 -45>, angularTarget=<1 0 0>}
      localJointOffset=<2 0 0>
     }]
    relativeSpaceBone {boneName=thigh_r}
   }
   Save cached pose 'chainCache'(Pose=Component To Local(ComponentPose=@n48)) {cacheName=chainCache}
  ]}
 EventGraph {nodes [
   Event OnParachuteOpened()
   > Switch on EGliderType(Selection=Get PlayerGliderType() -> (PlayerGliderType=HangGlider))
   > Set isHangGliding(isHangGliding:bool=true) -> (Output_Get:bool)
   Event OnCharacterCustomizationLoaded()
   > n1 = Cast To PlayerPawn_Athena(Object=TryGetPawnOwner())
   Event OnParachuteClosed()
   > Set isHangGliding(isHangGliding=false)
   GetSkeletalMeshForPartType(self=@n1, PartType=Head)
   Event BlueprintUpdateAnimation()
  ]}
]
variables [isHangGliding {type=boolean}]
