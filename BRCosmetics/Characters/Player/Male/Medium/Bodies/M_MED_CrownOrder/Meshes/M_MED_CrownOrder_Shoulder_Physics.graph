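# Physics asset for the shoulder mesh of the CrownOrder medium male player body, modelling two jointed chains on the right upper arm.
# Capsule bodies link upperarm_r to a base, which branches through small anchors into tapering upper (three-segment) and lower (four-segment) chains.
# Base and anchor joints swing and twist freely; each chain's first segment hinges on one swing axis, while later joints allow limited, mostly single-axis bending.
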PhysicsAsset /BRCosmetics/Characters/Player/Male/Medium/Bodies/M_MED_CrownOrder/Meshes/M_MED_CrownOrder_Shoulder_Physics
bodies {
 base_shoulderChain [{
   shapeType=Capsule
   length=13
   radius=3.5
   shapeName=base_shoulderChain_capsule
   center=<-6.7751 0 1.319>
   rotation=<rot 61.3442 -179.9332 -179.9488>
  }]
 base_shoulderChain_dw [{shapeType=Capsule, length=0.505, radius=0.505, shapeName=base_shoulderChain_dw_capsule, rotation=<rot 0 -90 0>}]
 base_shoulderChain_up [{shapeType=Capsule, length=0.505, radius=0.505, shapeName=base_shoulderChain_up_capsule, rotation=<rot 0 -90 0>}]
 dyn_shoulderChain_01_dw [{shapeType=Capsule, length=10.8559, radius=1.8, shapeName=dyn_shoulderChain_01_dw_capsule, center=<1.2266 0 0>, rotation=<rot 0 0 90>}]
 dyn_shoulderChain_01_up [{shapeType=Capsule, length=9, radius=1.4, shapeName=dyn_shoulderChain_01_up_capsule, center=<1.1863 0 0>, rotation=<rot 0 0 90>}]
 dyn_shoulderChain_02_dw [{shapeType=Capsule, length=9.3199, radius=1.6, shapeName=dyn_shoulderChain_02_dw_capsule, center=<1.5874 0 0>, rotation=<rot 0 0 90>}]
 dyn_shoulderChain_02_up [{shapeType=Capsule, length=6.5, radius=1, shapeName=dyn_shoulderChain_02_up_capsule, center=<1.2348 0 0>, rotation=<rot 0 0 90>}]
 dyn_shoulderChain_03_dw [{shapeType=Capsule, length=5.6517, radius=1.4, shapeName=dyn_shoulderChain_03_dw_capsule, center=<1.5347 0 0>, rotation=<rot 0 0 90>}]
 dyn_shoulderChain_03_up [{shapeType=Capsule, length=4, radius=0.7, shapeName=dyn_shoulderChain_03_up_capsule, center=<0.5577 0 0>, rotation=<rot 0 0 90>}]
 dyn_shoulderChain_04_dw [{shapeType=Capsule, length=3.6626, radius=1, shapeName=dyn_shoulderChain_04_dw_capsule, center=<0.725 0 0>, rotation=<rot 0 0 90>}]
 upperarm_r [{
   shapeType=Capsule
   length=13.5
   radius=6
   shapeName=upperarm_r_capsule
   center=<-7.808 1.2386 0.3771>
   rotation=<rot 88.2041 -177.6837 -179.9953>
  }]
}
constraints [
 {bone1Name=dyn_shoulderChain_03_up, bone2Name=dyn_shoulderChain_02_up, swing1LimitDegrees=5, swing2LimitDegrees=45, twistLimitDegrees=10}
 {
  bone1Name=dyn_shoulderChain_01_up
  bone2Name=base_shoulderChain_up
  swing1LimitDegrees=0
  swing1Motion=Locked
  swing2LimitDegrees=30
  twistLimitDegrees=0
  twistMotion=Locked
 }
 {bone1Name=dyn_shoulderChain_02_up, bone2Name=dyn_shoulderChain_01_up, swing1LimitDegrees=5, swing2LimitDegrees=45, twistLimitDegrees=10}
 {bone1Name=dyn_shoulderChain_04_dw, bone2Name=dyn_shoulderChain_03_dw, swing1LimitDegrees=5, swing2LimitDegrees=45, twistLimitDegrees=10}
 {
  bone1Name=dyn_shoulderChain_01_dw
  bone2Name=base_shoulderChain_dw
  swing1LimitDegrees=0
  swing1Motion=Locked
  swing2LimitDegrees=30
  twistLimitDegrees=0
  twistMotion=Locked
 }
 {bone1Name=dyn_shoulderChain_02_dw, bone2Name=dyn_shoulderChain_01_dw, swing1LimitDegrees=5, swing2LimitDegrees=45, twistLimitDegrees=10}
 {bone1Name=dyn_shoulderChain_03_dw, bone2Name=dyn_shoulderChain_02_dw, swing1LimitDegrees=5, swing2LimitDegrees=45, twistLimitDegrees=10}
 {
  bone1Name=base_shoulderChain_up
  bone2Name=base_shoulderChain
  swing1LimitDegrees=45
  swing1Motion=Free
  swing2LimitDegrees=45
  swing2Motion=Free
  twistLimitDegrees=45
  twistMotion=Free
 }
 {
  bone1Name=base_shoulderChain_dw
  bone2Name=base_shoulderChain
  swing1LimitDegrees=45
  swing1Motion=Free
  swing2LimitDegrees=45
  swing2Motion=Free
  twistLimitDegrees=45
  twistMotion=Free
 }
 {
  bone1Name=base_shoulderChain
  bone2Name=upperarm_r
  swing1LimitDegrees=45
  swing1Motion=Free
  swing2LimitDegrees=45
  swing2Motion=Free
  twistLimitDegrees=45
  twistMotion=Free
 }
]
