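# Animation Blueprint for the M_MED_CrownOrder face accessory part, which follows the character's body and head poses.
# When character customization loads, the event graph casts the owning pawn to PlayerPawn_Athena and stores its Head part mesh in HeadRef.
# The anim graph layers a pose copied from HeadRef over one copied from the attached parent using Layered blend per bone, then outputs it through the FaceAttach cached pose.
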AnimBlueprint /BRCosmetics/Characters/Player/Male/Medium/Bodies/M_MED_CrownOrder/Meshes/Parts/M_MED_CrownOrder_FaceAcc_AnimBP
target_skeleton=&/M_MED_CrownOrder_FaceAcc_Skeleton
graphs [
 AnimGraph {nodes [
   Output Pose(Result=Use cached pose 'FaceAttach'()) {layerGroup=None}
   n1 = Get HeadRef() -> (HeadRef:Skeletal Mesh Component Object Reference)
   Save cached pose 'FaceAttach'(Pose=Layered blend per bone(
     BasePose=Copy Pose From Mesh() {bUseAttachedParent=true}
     BlendPoses_0=Copy Pose From Mesh(SourceMeshComponent=@n1) {bUseAttachedParent=true}
    )) {cacheName=FaceAttach}
  ]}
 EventGraph {nodes [
   Event OnCharacterCustomizationLoaded()
   > n1 = Cast To PlayerPawn_Athena(Object=TryGetPawnOwner())
   > Set HeadRef(HeadRef:Skeletal Mesh Component Object Reference=GetSkeletalMeshForPartType(self=@n1, PartType=Head)) -> (Output_Get:Skeletal Mesh Component Object Reference)
  ]}
]
variables [HeadRef {type=SkeletalMeshComponent}]
